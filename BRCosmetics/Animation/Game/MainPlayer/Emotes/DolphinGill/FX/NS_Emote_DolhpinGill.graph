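# Niagara effect for the main player's DolphinGill emote, playing timed, mostly orange-gold lightning on a skeletal mesh.
# Delayed emitters sequence an arc buildup at one socket, appear and disappear phases, and an ending burst.
# Appear and disappear pair bone-flow ribbons with mesh arcs swept through bone indices, all reading a shared system mesh reader.
# Before the disappearance, one sprite-and-light particle steps through seven sockets, pulsing blue, yellow and red.
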NiagaraSystem /BRCosmetics/Animation/Game/MainPlayer/Emotes/DolphinGill/FX/NS_Emote_DolhpinGill
dependencies {
 usedDataInterfaces [
  &NiagaraDataInterfaceCurve
  &NiagaraDataInterfaceArrayInt32
  &NiagaraDataInterfaceColorCurve
  &NiagaraDataInterfaceSkeletalMesh
 ]
 usedDynamicInputs [&Vector2DFromFloat]
 usedModules [
  &self:SystemSpawnScript.SystemScriptSource.SystemScriptGraph.NiagaraNodeAssignment_1.SetVariables_#1
  &self:SystemSpawnScript.SystemScriptSource.SystemScriptGraph.NiagaraNodeAssignment_3.SetVariables_#2
  &SystemState
 ]
}
emitters {
 Appear_String_Lighting {
  emitterUpdateScript [
   &EmitterState(Life Cycle Mode=::NewEnumerator1, Loop Behavior=::NewEnumerator1, Loop Delay=2.75, Loop Duration=10, UseLoopDelay=true)
   &SpawnBurst_Instantaneous(Spawn Count=40, Use Spawn Probability=false, UseLoopCountLimit=false)
  ]
  particleSpawnScript [
   &FN_SetSpawnGroup(Inclusivity=true, Normalize Spawn Group=true, Particles Per Group=1, Pattern=::NewEnumerator0)
   &InitializeParticle(
    Color=<rgba 15 9.9172 0 1>
    Lifetime=0.9
    Ribbon Twist Mode:ENiagara_UnsetDirectSet=::NewEnumerator1
    Ribbon Width=10
    Ribbon Width Mode=::NewEnumerator1
    UsePositionOffset=false
    Write Material Random=true
   )
   &CameraOffset(Camera Offset Amount=10)
  ]
  particleUpdateScript [
   &ParticleState()
   &FN_BoneFlow(Enable Position Offsets=true, Radius=5, SkeletalMesh:NiagaraDataInterfaceSkeletalMesh=$System.Skeletal Mesh Reader) {module=FN_BoneFlow001}
   &DynamicMaterialParameters(
    Index 0 Param 1=&FloatFromCurve(FloatCurve=DataInterface() {propertyValues {Curve {
        defaultValue=0
        keys [
         {interpMode=RCIM_Cubic, tangentMode=RCTM_User, time=0, value=1}
         {arriveTangent=-3.0819, interpMode=RCIM_Cubic, leaveTangent=-3.0819, tangentMode=RCTM_User, time=1, value=0}
        ]
       }}})
    Index 0 Param 2=0.7
    Index 0 Param 3=&FloatFromCurve(FloatCurve=DataInterface() {propertyValues {Curve {keys [
         {arriveTangent=0.2581, interpMode=RCIM_Cubic, leaveTangent=0.2581, tangentMode=RCTM_User, time=0, value=0}
         {arriveTangent=2.8205, interpMode=RCIM_Cubic, leaveTangent=2.8205, tangentMode=RCTM_User, time=1, value=1}
        ]}}})
    Index 0 Param 4=3
    Index 1 Param 1=&FloatFromCurve(FloatCurve=DataInterface() {propertyValues {Curve {keys [<at 0 1>,<at 1 0.3>]}}})
    Module.Index0Float0Write=true
    Module.Index0Float1Write=true
    Module.Index0Float2Write=true
    Module.Index0Float3Write=true
    Module.Index1Float0Write=true
    Module.Index1Float1Write=true
    Module.Index1Float2Write=true
    Module.Index1Float3Write=true
    Write Parameter Index 1=::NewEnumerator1
   ) {module=DynamicMaterialParameters001}
   /SetVariables(Particles.Glow Color=&MakeLinearColorFromVectorAndFloat(
     "Vector (RGB)"=&MakeVectorFromLinearColorRGB(Color=<rgba 1 0.5505 0 1>)
     "Float (Alpha)"=&Multiply_Float(
      A=&FloatFromCurve(FloatCurve=DataInterface() {propertyValues {Curve {
          defaultValue=1
          keys [{interpMode=RCIM_Cubic, tangentMode=RCTM_User, time=0, value=1}, {interpMode=RCIM_Cubic, tangentMode=RCTM_User, time=1, value=0}]
         }}})
      B=2
     )
    )) {isSetParameters=true, module=SetVariables_#3}
   /SetVariables(Particles.Glow Dy02:Vector4f=&MakeVector4(
     X=&FloatFromCurve(FloatCurve=DataInterface() {propertyValues {Curve {
         defaultValue=0
         keys [
          {interpMode=RCIM_Cubic, tangentMode=RCTM_User, time=0, value=1}
          {arriveTangent=-3.0819, interpMode=RCIM_Cubic, leaveTangent=-3.0819, tangentMode=RCTM_User, time=1, value=0}
         ]
        }}})
     Y=0.7
     Z=0
     W=0
    )) {isSetParameters=true, module=SetVariables_#4}
   &ScaleRibbonWidth(Ribbon Width Scale=&FloatFromCurve(
     FloatCurve=DataInterface() {propertyValues {Curve {
        defaultValue=1
        keys [
         {interpMode=RCIM_Cubic, tangentMode=RCTM_User, time=0, value=0}
         {interpMode=RCIM_Cubic, tangentMode=RCTM_User, time=0.5, value=1}
         {tangentMode=RCTM_User, time=1, value=0}
        ]
       }}}
     Scale Curve=&Waveform("Waveform [1]"=::NewEnumerator8, "[1] Frequency"=1, "[1] Amplitude Min/Max"=<0.9 1.25>)
    ))
  ]
  rendererClasses [&NiagaraRibbonRendererProperties]
  renderers [
   {rendererClass=&NiagaraRibbonRendererProperties}
   {rendererClass=&NiagaraRibbonRendererProperties, rendererIndex=1}
  ]
 }
 Disappear_String_Lighting {
  emitterUpdateScript [
   &EmitterState(Life Cycle Mode=::NewEnumerator1, Loop Behavior=::NewEnumerator1, Loop Delay=9.6, Loop Duration=11, UseLoopDelay=true)
   &SpawnBurst_Instantaneous(Spawn Count=40, Use Spawn Probability=false, UseLoopCountLimit=false)
  ]
  particleSpawnScript [
   &FN_SetSpawnGroup(Inclusivity=true, Normalize Spawn Group=true, Particles Per Group=1, Pattern=::NewEnumerator0)
   &InitializeParticle(
    Color=<rgba 15 9.9172 0 1>
    Lifetime=1.2
    Ribbon Width=10
    Ribbon Width Mode=::NewEnumerator1
    UsePositionOffset=false
    Write Material Random=true
   )
   &CameraOffset(Camera Offset Amount=10)
  ]
  particleUpdateScript [
   &ParticleState()
   &FN_BoneFlow(Enable Position Offsets=true, Radius=5, SkeletalMesh:NiagaraDataInterfaceSkeletalMesh=$System.Skeletal Mesh Reader) {module=FN_BoneFlow001}
   &DynamicMaterialParameters(
    Index 0 Param 1=&FloatFromCurve(FloatCurve=DataInterface() {propertyValues {Curve {
        defaultValue=0.2
        keys [
         {arriveTangent=-2.6636, interpMode=RCIM_Cubic, leaveTangent=-2.6636, tangentMode=RCTM_User, time=0, value=1}
         {interpMode=RCIM_Cubic, tangentMode=RCTM_User, time=1, value=0}
        ]
       }}})
    Index 0 Param 2=0.7
    Index 0 Param 3=&FloatFromCurve(FloatCurve=DataInterface() {propertyValues {Curve {
        defaultValue=1
        keys [
         {interpMode=RCIM_Cubic, tangentMode=RCTM_User, time=0, value=0}
         {arriveTangent=2.5763, interpMode=RCIM_Cubic, leaveTangent=2.5763, tangentMode=RCTM_User, time=1, value=1}
        ]
       }}})
    Index 0 Param 4=3
    Index 1 Param 4=&FloatFromCurve(FloatCurve=DataInterface() {propertyValues {Curve {
        defaultValue=0
        keys [{interpMode=RCIM_Cubic, tangentMode=RCTM_User, time=0, value=0}, {interpMode=RCIM_Cubic, tangentMode=RCTM_User, time=1, value=1}]
       }}})
    Module.Index0Float0Write=true
    Module.Index0Float1Write=true
    Module.Index0Float2Write=true
    Module.Index0Float3Write=true
    Module.Index1Float0Write=true
    Module.Index1Float1Write=true
    Module.Index1Float2Write=true
    Module.Index1Float3Write=true
    Write Parameter Index 1=::NewEnumerator1
   ) {module=DynamicMaterialParameters001}
   /SetVariables(Particles.Glow Color=&MakeLinearColorFromVectorAndFloat(
     "Vector (RGB)"=&MakeVectorFromLinearColorRGB(Color=<rgba 1 0.5505 0 1>)
     "Float (Alpha)"=&Multiply_Float(
      A=&FloatFromCurve(FloatCurve=DataInterface() {propertyValues {Curve {
          defaultValue=1
          keys [{interpMode=RCIM_Cubic, tangentMode=RCTM_User, time=0, value=1}, {interpMode=RCIM_Cubic, tangentMode=RCTM_User, time=1, value=0}]
         }}})
      B=2
     )
    )) {isSetParameters=true, module=SetVariables_#3}
   /SetVariables(Particles.Glow Dy02:Vector4f=&MakeVector4(
     X=&FloatFromCurve(FloatCurve=DataInterface() {propertyValues {Curve {defaultValue=1, keys [{tangentMode=RCTM_User, time=0, value=-1}, {tangentMode=RCTM_User, time=1, value=1.2}]}}})
     Y=0.7
     Z=0
     W=0
    )) {isSetParameters=true, module=SetVariables_#4}
   &ScaleRibbonWidth(Ribbon Width Scale=&FloatFromCurve(
     FloatCurve=DataInterface() {propertyValues {Curve {
        defaultValue=1
        keys [
         {interpMode=RCIM_Cubic, tangentMode=RCTM_User, time=0, value=0}
         {interpMode=RCIM_Cubic, tangentMode=RCTM_User, time=0.5, value=1}
         {tangentMode=RCTM_User, time=1, value=0}
        ]
       }}}
     Scale Curve=&Waveform("Waveform [1]"=::NewEnumerator8, "[1] Frequency"=1, "[1] Amplitude Min/Max"=<1.25 1.7>)
    ))
   &ScaleColor(
    Scale Alpha=&FloatFromCurve(
     FloatCurve=DataInterface() {propertyValues {Curve {
        defaultValue=1
        keys [{interpMode=RCIM_Cubic, tangentMode=RCTM_User, time=0, value=1}, {interpMode=RCIM_Cubic, tangentMode=RCTM_User, time=1, value=0}]
       }}}
     Scale Curve=10
    )
    Scale RGB=&VectorFromFloat(Value=&FloatFromCurve(
      FloatCurve=DataInterface() {propertyValues {Curve {
         defaultValue=0
         keys [{interpMode=RCIM_Cubic, tangentMode=RCTM_User, time=0, value=0}, {interpMode=RCIM_Cubic, tangentMode=RCTM_User, time=1, value=1}]
        }}}
      CurveIndex=&ReturnNormalizedExecIndex(Normalized Index Scale=26.6593)
     ))
    ScaleA=true
    ScaleRGB=true
   )
  ]
  rendererClasses [&NiagaraRibbonRendererProperties]
  renderers [{rendererClass=&NiagaraRibbonRendererProperties}]
 }
 EndLighting {
  emitterUpdateScript [
   &EmitterState(Life Cycle Mode=::NewEnumerator1, Loop Behavior=::NewEnumerator1, Loop Delay=10.2, Loop Duration=0.29, UseLoopDelay=true)
   &SpawnBurst_Instantaneous(Spawn Count=5, Use Spawn Probability=false, UseLoopCountLimit=false)
   &SpawnRate(
    SpawnRate=&FloatFromCurve(
     FloatCurve=DataInterface() {propertyValues {Curve {keys [<at 0.2 0 RCIM_Cubic>, {arriveTangent=10.0142, interpMode=RCIM_Cubic, leaveTangent=0.286, tangentMode=RCTM_Break, time=0.25, value=1}]}}}
     CurveIndex=$Emitter.NormalizedLoopAge
     Scale Curve=35
    )
    Use Spawn Probability=false
   )
  ]
  particleSpawnScript [
   &InitializeParticle(
    AdjustAlpha=true
    AdjustHue=true
    AdjustSaturation=true
    AdjustValue=true
    Color=&RandomRangeLinearColor(Minimum=<rgba 1 0.7667 0 1>, Maximum=<rgba 1 0.2365 0 1>)
    Color Mode=::NewEnumerator3
    Hue Shift Range:Vector2f=<-0.02 -0.03>
    Lifetime=&RandomRangeFloat(Minimum=0.1, Maximum=0.4)
    Mesh Scale Mode=::NewEnumerator3
    Mesh Uniform Scale=&RandomRangeFloat(Minimum=0.15, Maximum=0.2)
    UsePositionOffset=false
    Write Material Random=true
   )
   &ShapeLocation(MaskSpawnGroup=false, Override Seed=false, Sphere Radius=10, UseOwnerScale=true)
   &InitialMeshOrientation(
    Mesh Orientation Mode=::NewEnumerator7
    Rotation=&RandomRangeVector(Minimum=<-0.3 -0.3 -1>, Maximum=<0.3 0.3 1>)
    Use Rotation Vector=true
   )
   &InitialMeshOrientation(Mesh Orientation Mode=::NewEnumerator7, Rotation=<0 0.25 0>, Use Rotation Vector=true) {module=InitialMeshOrientation001}
   &SkeletalMeshLocation(
    Bone / Socket Index=7
    MaskSpawnGroup=false
    Mesh Sampling Type=::NewEnumerator5
    Skeletal Mesh=$System.Skeletal Mesh Reader
    Socket Sampling Mode=::NewEnumerator5
   )
   &CameraOffset(Camera Offset Amount=10)
  ]
  particleUpdateScript [
   &ParticleState()
   &InheritSourceMovement(
    Applied Position Scale=&VectorFromFloat(Value=&FloatFromCurve(FloatCurve=DataInterface() {propertyValues {Curve {keys [{arriveTangent=-0.8713, interpMode=RCIM_Cubic, leaveTangent=-0.8713, tangentMode=RCTM_User, time=0, value=0.9}, <at 1 0>]}}}))
    ApplyForce=false
    ApplyPosition=true
    ApplyPositionFalloff=false
   )
   &ScaleMeshSize(Scale Factor=&VectorFromFloat(Value=&FloatFromCurve(FloatCurve=DataInterface() {propertyValues {Curve {keys [
          {arriveTangent=1.9115, interpMode=RCIM_Cubic, leaveTangent=1.9115, tangentMode=RCTM_User, time=0, value=0}
          {arriveTangent=0.0224, interpMode=RCIM_Cubic, leaveTangent=0.0224, tangentMode=RCTM_User, time=1, value=1.5}
         ]}}})))
   &ScaleMeshSize(Scale Factor=&RandomRangeVector(Minimum=<1 1 1.5>, Maximum=<1 1 2>)) {module=ScaleMeshSize001}
   &DynamicMaterialParameters(
    Index 0 Param 2=&RandomRangeFloat(Minimum=0, Maximum=0)
    Index 0 Param 4=13
    Index 1 Param 1=&FloatFromCurve(FloatCurve=DataInterface() {propertyValues {Curve {keys [<at 0 0.5>,<at 1 1>]}}})
    Index 1 Param 2=2
    Index 1 Param 3=&FloatFromCurve(FloatCurve=DataInterface() {propertyValues {Curve {keys [{interpMode=RCIM_Cubic, tangentMode=RCTM_Break, time=0.5, value=-0.5}, <at 1 0.1>]}}})
    Index 2 Param 1=&Multiply_Float(
     A=&FloatFromCurve(
      FloatCurve=DataInterface() {propertyValues {Curve {keys [{interpMode=RCIM_Cubic, leaveTangent=8.1085, tangentMode=RCTM_Break, time=0.8568, value=1}, {tangentMode=RCTM_User, time=1, value=0}]}}}
      Scale Curve=0.8
     )
     B=&Divide_Float(
      A=&MakeFloatFromVector(Vector=$Particles.Scale, Channel=::NewEnumerator0)
      B=&RandomRangeFloat(Minimum=10, Maximum=15)
     )
    )
    Index 2 Param 2=&FloatFromCurve(
     FloatCurve=DataInterface() {propertyValues {Curve {keys [
         {interpMode=RCIM_Cubic, tangentMode=RCTM_Break, time=0.7, value=0}
         {arriveTangent=-0.0859, interpMode=RCIM_Cubic, tangentMode=RCTM_Break, time=1, value=-1}
        ]}}}
     Scale Curve=50
    )
    Index 2 Param 3=&Waveform(
     Global Input Position=&RandomRangeFloat(Minimum=0, Maximum=15)
     Global Phase=$Particles.Age
     "Waveform [1]"=::NewEnumerator12
     "[1] Phase Scale"=2
     "[1] Frequency"=1
     "[1] Amplitude Min/Max"=<0 1>
    )
    Index 2 Param 4=25
    Index 3 Param 1=&FloatFromCurve(
     FloatCurve=DataInterface() {propertyValues {Curve {keys [
         {interpMode=RCIM_Cubic, leaveTangent=-0.0025, tangentMode=RCTM_Break, time=0, value=0}
         {arriveTangent=1.8528, interpMode=RCIM_Cubic, tangentMode=RCTM_Break, time=1, value=1}
        ]}}}
     Scale Curve=&RandomRangeFloat(Minimum=0, Maximum=0)
    )
    Index 3 Param 2=&FloatFromCurve(
     FloatCurve=DataInterface() {propertyValues {Curve {keys [
         {interpMode=RCIM_Cubic, leaveTangent=-0.0025, tangentMode=RCTM_Break, time=0, value=0.5}
         {arriveTangent=-0.0316, interpMode=RCIM_Cubic, tangentMode=RCTM_Break, time=1, value=-0.2}
        ]}}}
     Scale Curve=0
    )
    Index 3 Param 3=&FloatFromCurve(FloatCurve=DataInterface() {propertyValues {Curve {keys [
         {interpMode=RCIM_Cubic, leaveTangent=1.1034, tangentMode=RCTM_Break, time=0, value=-1}
         {arriveTangent=2.2399, tangentMode=RCTM_User, time=0.5, value=0.125}
        ]}}})
    Index 3 Param 4=&FloatFromCurve(FloatCurve=DataInterface() {propertyValues {Curve {keys [
         {interpMode=RCIM_Cubic, leaveTangent=1.8089, tangentMode=RCTM_Break, time=0.3, value=-0.5}
         {arriveTangent=2.0757, interpMode=RCIM_Cubic, tangentMode=RCTM_Break, time=0.9, value=1}
        ]}}})
    Module.Index0Param0Write=true
    Module.Index0Param1Write=true
    Module.Index0Param2Write=true
    Module.Index0Param3Write=true
    Module.Index1Param0Write=true
    Module.Index1Param1Write=true
    Module.Index1Param2Write=true
    Module.Index1Param3Write=true
    Module.Index2Param0Write=true
    Module.Index2Param1Write=true
    Module.Index2Param2Write=true
    Module.Index2Param3Write=true
    Module.Index3Param0Write=true
    Module.Index3Param1Write=true
    Module.Index3Param2Write=true
    Module.Index3Param3Write=true
    Write Parameter Index 1=true
    Write Parameter Index 2=true
    Write Parameter Index 3=true
   )
   &SolveForcesAndVelocity(Clamp Velocity=false, Limit Acceleration=false)
   &ScaleColor(
    Scale Alpha=&FloatFromCurve(FloatCurve=DataInterface() {propertyValues {Curve {keys [
         {tangentMode=RCTM_User, time=0, value=1}
         {arriveTangent=-4.3986, interpMode=RCIM_Cubic, leaveTangent=-4.3986, tangentMode=RCTM_User, time=0.45, value=1}
         {arriveTangent=0.1205, interpMode=RCIM_Cubic, leaveTangent=0.1205, tangentMode=RCTM_User, time=1, value=0}
        ]}}})
    Scale RGB=&Lerp_Vector(
     A=&VectorFromFloat(Value=&FloatFromCurve(FloatCurve=DataInterface() {propertyValues {Curve {keys [<at 0.2771 1>,<at 0.3786 1>]}}}, Scale Curve=40))
     B=&VectorFromFloat(Value=&RandomRangeFloat(Minimum=0, Maximum=2, Evaluation Type=::NewEnumerator1))
     Alpha=&FloatFromCurve(FloatCurve=DataInterface() {propertyValues {Curve {keys [
          {interpMode=RCIM_Cubic, leaveTangent=-0.3196, tangentMode=RCTM_Break, time=0, value=0}
          {arriveTangent=6.19, interpMode=RCIM_Cubic, tangentMode=RCTM_Break, time=0.4, value=1}
         ]}}})
    )
    ScaleA=true
    ScaleRGB=true
   )
  ]
  rendererClasses [&NiagaraMeshRendererProperties]
  renderers [
   {rendererClass=&NiagaraMeshRendererProperties, rendererIndex=0}
   {rendererClass=&NiagaraMeshRendererProperties, rendererIndex=1}
  ]
 }
 Light_Pops {
  emitterUpdateScript [
   &EmitterState(
    Inactive Response=::NewEnumerator1
    Life Cycle Mode=::NewEnumerator1
    Loop Behavior=::NewEnumerator1
    Loop Delay=8.1
    Loop Duration=1.7
    UseLoopDelay=true
   )
   &SpawnBurst_Instantaneous(Use Spawn Probability=false, UseLoopCountLimit=false)
  ]
  particleSpawnScript [
   &InitializeParticle(Lifetime=1.7, Sprite Size Mode=::NewEnumerator3, Uniform Sprite Size=60, UsePositionOffset=false, Write Material Random=true)
   &CameraOffset(Camera Offset Amount=5)
  ]
  particleUpdateScript [
   &ParticleState()
   &SkeletalMeshLocation(
    Bone / Socket Index=&MergeHelper_IntSwitch(
     0=0
     1=1
     2=2
     3=3
     4=4
     5=5
     6=6
     Index=&MakeIntFromFloat(
      Float=&Multiply_Float(A=&ClampFloat(Float=$Emitter.NormalizedLoopAge, Max=0.99), B=7)
      Conversion Mode=::NewEnumerator3
     )
     Num=::NewEnumerator5
    )
    MaskSpawnGroup=false
    Mesh Sampling Type=::NewEnumerator5
    Skeletal Mesh=$System.Skeletal Mesh Reader
    Socket Sampling Mode=::NewEnumerator5
   )
   &ScaleSpriteSize(
    Uniform Curve Index=&Frac_Float(Float=&Multiply_Float(A=$Particles.NormalizedAge, B=7))
    Uniform Curve Sprite Scale=DataInterface() {propertyValues {Curve {
       defaultValue=0
       keys [
        {interpMode=RCIM_Cubic, tangentMode=RCTM_User, time=0, value=0}
        {interpMode=RCIM_Cubic, tangentMode=RCTM_User, time=0.5, value=1}
        {tangentMode=RCTM_User, time=1, value=0}
       ]
      }}}
   )
   &ScaleColor(
    Scale Alpha=&Multiply_Float(
     A=0.5
     B=&Sine(
      Normalized Angle=&Multiply_Float(A=7, B=&ClampFloat(Float=$Emitter.NormalizedLoopAge, Max=0.99))
      Period=1
      Scale=1
      Bias=0
     )
    )
    Scale RGB=&Multiply_VectorByFloat(
     Vector=&MakeVectorFromLinearColorRGB(Color=&MergeHelper_ColorSwitch(
       0=<rgba 0.1 0.2426 1 1>
       1=<rgba 1 0.79 0.1 1>
       2=<rgba 1 0.79 0.1 1>
       3=<rgba 1 0.1457 0.1 1>
       4=<rgba 1 0.1457 0.1 1>
       5=<rgba 1 0.79 0.1 1>
       6=<rgba 0.1 0.2426 1 1>
       Index=&MakeIntFromFloat(
        Float=&Multiply_Float(A=&ClampFloat(Float=$Emitter.NormalizedLoopAge, Max=0.99), B=7)
        Conversion Mode=::NewEnumerator3
       )
       Num=::NewEnumerator5
      ))
     Float=&FloatFromCurve(
      FloatCurve=DataInterface() {propertyValues {Curve {keys [{arriveTangent=-12.7049, interpMode=RCIM_Cubic, leaveTangent=-12.7049, tangentMode=RCTM_User, time=0, value=10}, <at 1 1>]}}}
      CurveIndex=&Frac_Float(Float=&Multiply_Float(A=$Particles.NormalizedAge, B=7))
     )
    )
    ScaleA=true
    ScaleRGB=true
   )
   &FN_Light_Attributes(
    Color=$Particles.Color
    Create Color Attribute=true
    Create Exponent Attribute=true
    Create Position Attribute=true
    Create Radius Attribute=true
    Exponent=1
    Position=&AddVectorToPosition(
     Position=$Particles.Position
     Vector=&MergeHelper_VectorSwitch(
      0=<0 -5 4>
      1=<3 -3 -2>
      2=<-3 3 4>
      3=<3 0 4>
      4=<2 -2 0>
      5=<2 0 5>
      6=<3 0 4>
      Index=&MakeIntFromFloat(
       Float=&Multiply_Float(A=&ClampFloat(Float=$Emitter.NormalizedLoopAge, Max=0.99), B=7)
       Conversion Mode=::NewEnumerator3
      )
      Num=::NewEnumerator5
     )
    )
    Radius=&FloatFromCurve(
     FloatCurve=DataInterface() {propertyValues {Curve {
        defaultValue=1
        keys [
         {interpMode=RCIM_Cubic, tangentMode=RCTM_User, time=0, value=0}
         {interpMode=RCIM_Cubic, tangentMode=RCTM_User, time=0.5, value=1}
         {tangentMode=RCTM_User, time=1, value=0}
        ]
       }}}
     CurveIndex=&Frac_Float(Float=&Multiply_Float(A=$Particles.NormalizedAge, B=7))
     Scale Curve=20
    )
    Scale Color=15
   )
  ]
  rendererClasses [&NiagaraSpriteRendererProperties, &NiagaraLightRendererProperties]
  renderers [
   {rendererClass=&NiagaraSpriteRendererProperties}
   {rendererClass=&NiagaraLightRendererProperties, rendererIndex=1}
  ]
 }
 LightingArcs_Appear {
  emitterUpdateScript [
   &EmitterState(Life Cycle Mode=::NewEnumerator1, Loop Behavior=::NewEnumerator1, Loop Delay=2.66, Loop Duration=0.7, UseLoopDelay=true)
   &SpawnRate(SpawnRate=88, Use Spawn Probability=false)
  ]
  particleSpawnScript [
   &InitializeParticle(
    AdjustAlpha=true
    AdjustHue=true
    AdjustSaturation=true
    AdjustValue=true
    Color=&MergeHelper_ColorSwitch(
     0=<rgba 1 0.2817 0 1>
     1=<rgba 1 0.7667 0 1>
     2=?
     3=?
     Index=&RandomRangeInt(Minimum=0, Maximum=1)
     Num=::NewEnumerator0
    )
    Color Mode=::NewEnumerator3
    Hue Shift Range:Vector2f=<-0.01 0.01>
    Lifetime Max=0.22
    Lifetime Min=0.18
    Lifetime Mode=::NewEnumerator1
    Mesh Scale Mode=::NewEnumerator3
    Mesh Uniform Scale=&RandomRangeFloat(Minimum=0.1, Maximum=0.12)
    UsePositionOffset=false
    Write Material Random=true
   )
   &ShapeLocation(MaskSpawnGroup=false, Override Seed=false, Sphere Radius=5, UseOwnerScale=true)
   &InitialMeshOrientation(
    Mesh Orientation Mode=::NewEnumerator7
    Rotation=&RandomRangeVector(Minimum=<-0.3 -0.3 -1>, Maximum=<0.3 0.3 1>)
    Use Rotation Vector=true
   )
   &InitialMeshOrientation(Mesh Orientation Mode=::NewEnumerator7, Rotation=<0 0.25 0>, Use Rotation Vector=true) {module=InitialMeshOrientation001}
   &CameraOffset(Camera Offset Amount=10)
   &SkeletalMeshLocation(
    Bone / Socket Index=&MakeIntFromFloat(
     Float=&Multiply_Float(
      A=44
      B=&Multiply_Float(
       A=&FloatFromCurve(
        FloatCurve=DataInterface() {propertyValues {Curve {
           defaultValue=0
           keys [{interpMode=RCIM_Cubic, tangentMode=RCTM_User, time=0, value=1}, {interpMode=RCIM_Cubic, tangentMode=RCTM_User, time=1, value=0}]
          }}}
        CurveIndex=$Emitter.NormalizedLoopAge
       )
       B=1
      )
     )
     Conversion Mode=::NewEnumerator3
    )
    Bone Sampling Mode=::NewEnumerator4
    MaskSpawnGroup=false
    Skeletal Mesh=$System.Skeletal Mesh Reader
   )
  ]
  particleUpdateScript [
   &ParticleState()
   &ScaleMeshSize(Scale Factor=&VectorFromFloat(Value=&FloatFromCurve(FloatCurve=DataInterface() {propertyValues {Curve {keys [
          {arriveTangent=1.9115, interpMode=RCIM_Cubic, leaveTangent=1.9115, tangentMode=RCTM_User, time=0, value=0}
          {arriveTangent=0.0224, interpMode=RCIM_Cubic, leaveTangent=0.0224, tangentMode=RCTM_User, time=1, value=1.5}
         ]}}})))
   &ScaleMeshSize(Scale Factor=&RandomRangeVector(Minimum=<1 1 1.5>, Maximum=<1 1 2>)) {module=ScaleMeshSize001}
   &DynamicMaterialParameters(
    Index 0 Param 2=&RandomRangeFloat(Minimum=0, Maximum=0)
    Index 0 Param 4=13
    Index 1 Param 1=&FloatFromCurve(FloatCurve=DataInterface() {propertyValues {Curve {keys [<at 0 0.5>,<at 1 1>]}}})
    Index 1 Param 2=2
    Index 1 Param 3=&FloatFromCurve(FloatCurve=DataInterface() {propertyValues {Curve {keys [{interpMode=RCIM_Cubic, tangentMode=RCTM_Break, time=0.5, value=-0.5}, <at 1 0.1>]}}})
    Index 2 Param 1=&Multiply_Float(
     A=&FloatFromCurve(
      FloatCurve=DataInterface() {propertyValues {Curve {keys [{interpMode=RCIM_Cubic, leaveTangent=8.1085, tangentMode=RCTM_Break, time=0.8568, value=1}, {tangentMode=RCTM_User, time=1, value=0}]}}}
      Scale Curve=0.8
     )
     B=&Divide_Float(
      A=&MakeFloatFromVector(Vector=$Particles.Scale, Channel=::NewEnumerator0)
      B=&RandomRangeFloat(Minimum=10, Maximum=15)
     )
    )
    Index 2 Param 2=&FloatFromCurve(
     FloatCurve=DataInterface() {propertyValues {Curve {keys [
         {interpMode=RCIM_Cubic, tangentMode=RCTM_Break, time=0.7, value=0}
         {arriveTangent=-0.0859, interpMode=RCIM_Cubic, tangentMode=RCTM_Break, time=1, value=-1}
        ]}}}
     Scale Curve=50
    )
    Index 2 Param 3=&Waveform(
     Global Input Position=&RandomRangeFloat(Minimum=0, Maximum=15)
     Global Phase=$Particles.Age
     "Waveform [1]"=::NewEnumerator12
     "[1] Phase Scale"=2
     "[1] Frequency"=1
     "[1] Amplitude Min/Max"=<0 1>
    )
    Index 2 Param 4=25
    Index 3 Param 1=&FloatFromCurve(
     FloatCurve=DataInterface() {propertyValues {Curve {keys [
         {interpMode=RCIM_Cubic, leaveTangent=-0.0025, tangentMode=RCTM_Break, time=0, value=0}
         {arriveTangent=1.8528, interpMode=RCIM_Cubic, tangentMode=RCTM_Break, time=1, value=1}
        ]}}}
     Scale Curve=&RandomRangeFloat(Minimum=0, Maximum=0)
    )
    Index 3 Param 2=&FloatFromCurve(
     FloatCurve=DataInterface() {propertyValues {Curve {keys [
         {interpMode=RCIM_Cubic, leaveTangent=-0.0025, tangentMode=RCTM_Break, time=0, value=0.5}
         {arriveTangent=-0.0316, interpMode=RCIM_Cubic, tangentMode=RCTM_Break, time=1, value=-0.2}
        ]}}}
     Scale Curve=0
    )
    Index 3 Param 3=&FloatFromCurve(FloatCurve=DataInterface() {propertyValues {Curve {keys [
         {interpMode=RCIM_Cubic, leaveTangent=1.1034, tangentMode=RCTM_Break, time=0, value=-1}
         {arriveTangent=2.2399, tangentMode=RCTM_User, time=0.5, value=0.125}
        ]}}})
    Index 3 Param 4=&FloatFromCurve(FloatCurve=DataInterface() {propertyValues {Curve {keys [
         {interpMode=RCIM_Cubic, leaveTangent=1.8089, tangentMode=RCTM_Break, time=0.3, value=-0.5}
         {arriveTangent=2.0757, interpMode=RCIM_Cubic, tangentMode=RCTM_Break, time=0.9, value=1}
        ]}}})
    Module.Index0Param0Write=true
    Module.Index0Param1Write=true
    Module.Index0Param2Write=true
    Module.Index0Param3Write=true
    Module.Index1Param0Write=true
    Module.Index1Param1Write=true
    Module.Index1Param2Write=true
    Module.Index1Param3Write=true
    Module.Index2Param0Write=true
    Module.Index2Param1Write=true
    Module.Index2Param2Write=true
    Module.Index2Param3Write=true
    Module.Index3Param0Write=true
    Module.Index3Param1Write=true
    Module.Index3Param2Write=true
    Module.Index3Param3Write=true
    Write Parameter Index 1=true
    Write Parameter Index 2=true
    Write Parameter Index 3=true
   )
   &SolveForcesAndVelocity(Clamp Velocity=false, Limit Acceleration=false)
   &ScaleColor(
    Scale Alpha=&FloatFromCurve(FloatCurve=DataInterface() {propertyValues {Curve {keys [
         {tangentMode=RCTM_User, time=0, value=1}
         {arriveTangent=-4.3986, interpMode=RCIM_Cubic, leaveTangent=-4.3986, tangentMode=RCTM_User, time=0.45, value=1}
         {arriveTangent=0.1205, interpMode=RCIM_Cubic, leaveTangent=0.1205, tangentMode=RCTM_User, time=1, value=0}
        ]}}})
    Scale RGB=&Lerp_Vector(
     A=&VectorFromFloat(Value=&FloatFromCurve(FloatCurve=DataInterface() {propertyValues {Curve {keys [<at 0.2771 1>,<at 0.3786 1>]}}}, Scale Curve=40))
     B=&VectorFromFloat(Value=&RandomRangeFloat(Minimum=0, Maximum=2, Evaluation Type=::NewEnumerator1))
     Alpha=&FloatFromCurve(FloatCurve=DataInterface() {propertyValues {Curve {keys [
          {interpMode=RCIM_Cubic, leaveTangent=-0.3196, tangentMode=RCTM_Break, time=0, value=0}
          {arriveTangent=6.19, interpMode=RCIM_Cubic, tangentMode=RCTM_Break, time=0.4, value=1}
         ]}}})
    )
    ScaleA=true
    ScaleRGB=true
   )
  ]
  rendererClasses [&NiagaraMeshRendererProperties]
  renderers [
   {rendererClass=&NiagaraMeshRendererProperties, rendererIndex=0}
   {rendererClass=&NiagaraMeshRendererProperties, rendererIndex=1}
  ]
 }
 Lighting_Arcs_Buildup {
  emitterUpdateScript [
   &EmitterState(Life Cycle Mode=::NewEnumerator1, Loop Behavior=::NewEnumerator1, Loop Duration=2.8, UseLoopDelay=false)
   &SpawnRate(
    SpawnRate=&Multiply_Float(
     A=&FloatFromCurve(
      FloatCurve=DataInterface() {propertyValues {Curve {
         defaultValue=1
         keys [
          {interpMode=RCIM_Cubic, tangentMode=RCTM_User, time=0, value=0}
          {arriveTangent=1.393, interpMode=RCIM_Cubic, leaveTangent=1.393, tangentMode=RCTM_User, time=1, value=0.75}
         ]
        }}}
      CurveIndex=$Emitter.NormalizedLoopAge
      Scale Curve=100
     )
     B=1
    )
    Use Spawn Probability=false
   )
  ]
  particleSpawnScript [
   &InitializeParticle(
    AdjustAlpha=true
    AdjustHue=true
    AdjustSaturation=true
    AdjustValue=true
    Color=&MergeHelper_ColorSwitch(
     0=<rgba 1 0.2817 0 1>
     1=<rgba 1 0.7667 0 1>
     2=?
     3=?
     Index=&RandomRangeInt(Minimum=0, Maximum=1)
     Num=::NewEnumerator0
    )
    Color Mode=::NewEnumerator3
    Hue Shift Range:Vector2f=<-0.01 0.01>
    Lifetime Max=0.6
    Lifetime Min=0.3
    Lifetime Mode=::NewEnumerator1
    Mesh Scale Mode=::NewEnumerator3
    Mesh Uniform Scale=&RandomRangeFloat(Minimum=0.05, Maximum=0.08)
    UsePositionOffset=false
    Write Material Random=true
   )
   &SkeletalMeshLocation(
    Bone / Socket Index=7
    MaskSpawnGroup=false
    Mesh Sampling Type=::NewEnumerator5
    Skeletal Mesh=$System.Skeletal Mesh Reader
    Socket Sampling Mode=::NewEnumerator5
   )
   &ShapeLocation(MaskSpawnGroup=false, Override Seed=false, Sphere Radius=7, UseOwnerScale=true)
   &InitialMeshOrientation(
    Mesh Orientation Mode=::NewEnumerator7
    Rotation=&RandomRangeVector(Minimum=<-0.3 -0.3 -1>, Maximum=<0.3 0.3 1>)
    Use Rotation Vector=true
   )
   &InitialMeshOrientation(Mesh Orientation Mode=::NewEnumerator7, Rotation=<0 0.25 0>, Use Rotation Vector=true) {module=InitialMeshOrientation001}
   &CameraOffset(Camera Offset Amount=5)
   &AddVelocity(
    Module.Use Falloff Distance=false
    Override Seed=false
    Velocity Mode=::NewEnumerator1
    Velocity Speed=&Multiply_Float(
     A=&RandomRangeFloat(Minimum=30, Maximum=50)
     B=&Waveform(
      "Waveform [1]"=::NewEnumerator5
      "[1] Phase Scale"=&FloatFromCurve(
       FloatCurve=DataInterface() {propertyValues {Curve {keys [<at 0 1 RCIM_Cubic>, {arriveTangent=4.308, interpMode=RCIM_Cubic, leaveTangent=4.308, tangentMode=RCTM_User, time=1.0012, value=4}]}}}
       CurveIndex=$Emitter.NormalizedLoopAge
      )
      "[1] Frequency"=1
      "[1] Amplitude Min/Max"=<0 1>
     )
    )
   )
  ]
  particleUpdateScript [
   &ParticleState()
   &ScaleMeshSize(Scale Factor=&VectorFromFloat(Value=&FloatFromCurve(
      FloatCurve=DataInterface() {propertyValues {Curve {keys [
          {arriveTangent=1.9115, interpMode=RCIM_Cubic, leaveTangent=1.9115, tangentMode=RCTM_User, time=0, value=0}
          {arriveTangent=0.0224, interpMode=RCIM_Cubic, leaveTangent=0.0224, tangentMode=RCTM_User, time=1, value=1.5}
         ]}}}
      Scale Curve=1.25
     )))
   &DynamicMaterialParameters(
    Index 0 Param 2=&RandomRangeFloat(Minimum=0, Maximum=0)
    Index 0 Param 4=13
    Index 1 Param 1=&FloatFromCurve(FloatCurve=DataInterface() {propertyValues {Curve {keys [<at 0 0.5>,<at 1 1>]}}})
    Index 1 Param 2=2
    Index 1 Param 3=&FloatFromCurve(FloatCurve=DataInterface() {propertyValues {Curve {keys [{interpMode=RCIM_Cubic, tangentMode=RCTM_Break, time=0.5, value=-0.5}, <at 1 0.1>]}}})
    Index 2 Param 1=&Multiply_Float(
     A=&FloatFromCurve(
      FloatCurve=DataInterface() {propertyValues {Curve {keys [{interpMode=RCIM_Cubic, leaveTangent=8.1085, tangentMode=RCTM_Break, time=0.8568, value=1}, {tangentMode=RCTM_User, time=1, value=0}]}}}
      Scale Curve=0.3
     )
     B=&Divide_Float(
      A=&MakeFloatFromVector(Vector=$Particles.Scale, Channel=::NewEnumerator0)
      B=&RandomRangeFloat(Minimum=10, Maximum=15)
     )
    )
    Index 2 Param 2=&FloatFromCurve(
     FloatCurve=DataInterface() {propertyValues {Curve {keys [
         {interpMode=RCIM_Cubic, tangentMode=RCTM_Break, time=0.7, value=0}
         {arriveTangent=-0.0859, interpMode=RCIM_Cubic, tangentMode=RCTM_Break, time=1, value=-1}
        ]}}}
     Scale Curve=50
    )
    Index 2 Param 3=&Waveform(
     Global Input Position=&RandomRangeFloat(Minimum=0, Maximum=15)
     Global Phase=$Particles.Age
     "Waveform [1]"=::NewEnumerator12
     "[1] Phase Scale"=2
     "[1] Frequency"=1
     "[1] Amplitude Min/Max"=<0 1>
    )
    Index 2 Param 4=25
    Index 3 Param 1=&FloatFromCurve(
     FloatCurve=DataInterface() {propertyValues {Curve {keys [
         {interpMode=RCIM_Cubic, leaveTangent=-0.0025, tangentMode=RCTM_Break, time=0, value=0}
         {arriveTangent=1.8528, interpMode=RCIM_Cubic, tangentMode=RCTM_Break, time=1, value=1}
        ]}}}
     Scale Curve=&RandomRangeFloat(Minimum=0, Maximum=0)
    )
    Index 3 Param 2=&FloatFromCurve(
     FloatCurve=DataInterface() {propertyValues {Curve {keys [
         {interpMode=RCIM_Cubic, leaveTangent=-0.0025, tangentMode=RCTM_Break, time=0, value=0.5}
         {arriveTangent=-0.0316, interpMode=RCIM_Cubic, tangentMode=RCTM_Break, time=1, value=-0.2}
        ]}}}
     Scale Curve=0
    )
    Index 3 Param 3=&FloatFromCurve(FloatCurve=DataInterface() {propertyValues {Curve {keys [
         {interpMode=RCIM_Cubic, leaveTangent=1.1034, tangentMode=RCTM_Break, time=0, value=-1}
         {arriveTangent=2.2399, tangentMode=RCTM_User, time=0.5, value=0.125}
        ]}}})
    Index 3 Param 4=&FloatFromCurve(FloatCurve=DataInterface() {propertyValues {Curve {keys [
         {interpMode=RCIM_Cubic, leaveTangent=1.8089, tangentMode=RCTM_Break, time=0.3, value=-0.5}
         {arriveTangent=2.0757, interpMode=RCIM_Cubic, tangentMode=RCTM_Break, time=0.9, value=1}
        ]}}})
    Module.Index0Param0Write=true
    Module.Index0Param1Write=true
    Module.Index0Param2Write=true
    Module.Index0Param3Write=true
    Module.Index1Param0Write=true
    Module.Index1Param1Write=true
    Module.Index1Param2Write=true
    Module.Index1Param3Write=true
    Module.Index2Param0Write=true
    Module.Index2Param1Write=true
    Module.Index2Param2Write=true
    Module.Index2Param3Write=true
    Module.Index3Param0Write=true
    Module.Index3Param1Write=true
    Module.Index3Param2Write=true
    Module.Index3Param3Write=true
    Write Parameter Index 1=true
    Write Parameter Index 2=true
    Write Parameter Index 3=true
   )
   &Drag(Drag=3, Rotational Drag=0, Use Linear Drag=true, Use Rotational Drag=true)
   &InheritSourceMovement(ApplyForce=false, ApplyPosition=true, ApplyPositionFalloff=false)
   &SolveForcesAndVelocity(Clamp Velocity=false, Limit Acceleration=false)
   &ScaleColor(
    Scale Alpha=&FloatFromCurve(FloatCurve=DataInterface() {propertyValues {Curve {keys [
         {tangentMode=RCTM_User, time=0, value=1}
         {arriveTangent=-4.3986, interpMode=RCIM_Cubic, leaveTangent=-4.3986, tangentMode=RCTM_User, time=0.45, value=1}
         {arriveTangent=0.1205, interpMode=RCIM_Cubic, leaveTangent=0.1205, tangentMode=RCTM_User, time=1, value=0}
        ]}}})
    Scale RGB=&Lerp_Vector(
     A=&VectorFromFloat(Value=&FloatFromCurve(FloatCurve=DataInterface() {propertyValues {Curve {keys [<at 0.2771 1>,<at 0.3786 1>]}}}, Scale Curve=40))
     B=&VectorFromFloat(Value=&RandomRangeFloat(Minimum=0, Maximum=2, Evaluation Type=::NewEnumerator1))
     Alpha=&FloatFromCurve(FloatCurve=DataInterface() {propertyValues {Curve {keys [
          {interpMode=RCIM_Cubic, leaveTangent=-0.3196, tangentMode=RCTM_Break, time=0, value=0}
          {arriveTangent=6.19, interpMode=RCIM_Cubic, tangentMode=RCTM_Break, time=0.4, value=1}
         ]}}})
    )
    ScaleA=true
    ScaleRGB=true
   )
  ]
  rendererClasses [&NiagaraMeshRendererProperties]
  renderers [
   {rendererClass=&NiagaraMeshRendererProperties, rendererIndex=0}
   {rendererClass=&NiagaraMeshRendererProperties, rendererIndex=1}
  ]
 }
 Lighting_Arcs_Disappear001 {
  emitterUpdateScript [
   &EmitterState(Life Cycle Mode=::NewEnumerator1, Loop Behavior=::NewEnumerator1, Loop Delay=9.6, Loop Duration=0.6, UseLoopDelay=true)
   &SpawnRate(SpawnRate=88, Use Spawn Probability=false)
  ]
  particleSpawnScript [
   &InitializeParticle(
    AdjustAlpha=true
    AdjustHue=true
    AdjustSaturation=true
    AdjustValue=true
    Color=&MergeHelper_ColorSwitch(
     0=<rgba 1 0.2817 0 1>
     1=<rgba 1 0.7667 0 1>
     2=?
     3=?
     Index=&RandomRangeInt(Minimum=0, Maximum=1)
     Num=::NewEnumerator0
    )
    Color Mode=::NewEnumerator3
    Hue Shift Range:Vector2f=<-0.01 0.01>
    Lifetime Max=0.22
    Lifetime Min=0.18
    Lifetime Mode=::NewEnumerator1
    Mesh Scale Mode=::NewEnumerator3
    Mesh Uniform Scale=&RandomRangeFloat(Minimum=0.1, Maximum=0.12)
    UsePositionOffset=false
    Write Material Random=true
   )
   &ShapeLocation(MaskSpawnGroup=false, Override Seed=false, Sphere Radius=5, UseOwnerScale=true)
   &InitialMeshOrientation(
    Mesh Orientation Mode=::NewEnumerator7
    Rotation=&RandomRangeVector(Minimum=<-0.3 -0.3 -1>, Maximum=<0.3 0.3 1>)
    Use Rotation Vector=true
   )
   &InitialMeshOrientation(Mesh Orientation Mode=::NewEnumerator7, Rotation=<0 0.25 0>, Use Rotation Vector=true) {module=InitialMeshOrientation001}
   &CameraOffset(Camera Offset Amount=10)
   &SkeletalMeshLocation(
    Bone / Socket Index=&MakeIntFromFloat(
     Float=&Multiply_Float(
      A=44
      B=&Multiply_Float(
       A=&FloatFromCurve(
        FloatCurve=DataInterface() {propertyValues {Curve {
           defaultValue=1
           keys [{interpMode=RCIM_Cubic, tangentMode=RCTM_User, time=0, value=0.03}, {interpMode=RCIM_Cubic, tangentMode=RCTM_User, time=1, value=1}]
          }}}
        CurveIndex=$Emitter.NormalizedLoopAge
       )
       B=1
      )
     )
     Conversion Mode=::NewEnumerator3
    )
    Bone Sampling Mode=::NewEnumerator4
    MaskSpawnGroup=false
    Skeletal Mesh=$System.Skeletal Mesh Reader
   )
  ]
  particleUpdateScript [
   &ParticleState()
   &ScaleMeshSize(Scale Factor=&VectorFromFloat(Value=&FloatFromCurve(FloatCurve=DataInterface() {propertyValues {Curve {keys [
          {arriveTangent=1.9115, interpMode=RCIM_Cubic, leaveTangent=1.9115, tangentMode=RCTM_User, time=0, value=0}
          {arriveTangent=0.0224, interpMode=RCIM_Cubic, leaveTangent=0.0224, tangentMode=RCTM_User, time=1, value=1.5}
         ]}}})))
   &ScaleMeshSize(Scale Factor=&RandomRangeVector(Minimum=<1 1 1.5>, Maximum=<1 1 2>)) {module=ScaleMeshSize001}
   &DynamicMaterialParameters(
    Index 0 Param 2=&RandomRangeFloat(Minimum=0, Maximum=0)
    Index 0 Param 4=13
    Index 1 Param 1=&FloatFromCurve(FloatCurve=DataInterface() {propertyValues {Curve {keys [<at 0 0.5>,<at 1 1>]}}})
    Index 1 Param 2=2
    Index 1 Param 3=&FloatFromCurve(FloatCurve=DataInterface() {propertyValues {Curve {keys [{interpMode=RCIM_Cubic, tangentMode=RCTM_Break, time=0.5, value=-0.5}, <at 1 0.1>]}}})
    Index 2 Param 1=&Multiply_Float(
     A=&FloatFromCurve(
      FloatCurve=DataInterface() {propertyValues {Curve {keys [{interpMode=RCIM_Cubic, leaveTangent=8.1085, tangentMode=RCTM_Break, time=0.8568, value=1}, {tangentMode=RCTM_User, time=1, value=0}]}}}
      Scale Curve=0.8
     )
     B=&Divide_Float(
      A=&MakeFloatFromVector(Vector=$Particles.Scale, Channel=::NewEnumerator0)
      B=&RandomRangeFloat(Minimum=10, Maximum=15)
     )
    )
    Index 2 Param 2=&FloatFromCurve(
     FloatCurve=DataInterface() {propertyValues {Curve {keys [
         {interpMode=RCIM_Cubic, tangentMode=RCTM_Break, time=0.7, value=0}
         {arriveTangent=-0.0859, interpMode=RCIM_Cubic, tangentMode=RCTM_Break, time=1, value=-1}
        ]}}}
     Scale Curve=50
    )
    Index 2 Param 3=&Waveform(
     Global Input Position=&RandomRangeFloat(Minimum=0, Maximum=15)
     Global Phase=$Particles.Age
     "Waveform [1]"=::NewEnumerator12
     "[1] Phase Scale"=2
     "[1] Frequency"=1
     "[1] Amplitude Min/Max"=<0 1>
    )
    Index 2 Param 4=25
    Index 3 Param 1=&FloatFromCurve(
     FloatCurve=DataInterface() {propertyValues {Curve {keys [
         {interpMode=RCIM_Cubic, leaveTangent=-0.0025, tangentMode=RCTM_Break, time=0, value=0}
         {arriveTangent=1.8528, interpMode=RCIM_Cubic, tangentMode=RCTM_Break, time=1, value=1}
        ]}}}
     Scale Curve=&RandomRangeFloat(Minimum=0, Maximum=0)
    )
    Index 3 Param 2=&FloatFromCurve(
     FloatCurve=DataInterface() {propertyValues {Curve {keys [
         {interpMode=RCIM_Cubic, leaveTangent=-0.0025, tangentMode=RCTM_Break, time=0, value=0.5}
         {arriveTangent=-0.0316, interpMode=RCIM_Cubic, tangentMode=RCTM_Break, time=1, value=-0.2}
        ]}}}
     Scale Curve=0
    )
    Index 3 Param 3=&FloatFromCurve(FloatCurve=DataInterface() {propertyValues {Curve {keys [
         {interpMode=RCIM_Cubic, leaveTangent=1.1034, tangentMode=RCTM_Break, time=0, value=-1}
         {arriveTangent=2.2399, tangentMode=RCTM_User, time=0.5, value=0.125}
        ]}}})
    Index 3 Param 4=&FloatFromCurve(FloatCurve=DataInterface() {propertyValues {Curve {keys [
         {interpMode=RCIM_Cubic, leaveTangent=1.8089, tangentMode=RCTM_Break, time=0.3, value=-0.5}
         {arriveTangent=2.0757, interpMode=RCIM_Cubic, tangentMode=RCTM_Break, time=0.9, value=1}
        ]}}})
    Module.Index0Param0Write=true
    Module.Index0Param1Write=true
    Module.Index0Param2Write=true
    Module.Index0Param3Write=true
    Module.Index1Param0Write=true
    Module.Index1Param1Write=true
    Module.Index1Param2Write=true
    Module.Index1Param3Write=true
    Module.Index2Param0Write=true
    Module.Index2Param1Write=true
    Module.Index2Param2Write=true
    Module.Index2Param3Write=true
    Module.Index3Param0Write=true
    Module.Index3Param1Write=true
    Module.Index3Param2Write=true
    Module.Index3Param3Write=true
    Write Parameter Index 1=true
    Write Parameter Index 2=true
    Write Parameter Index 3=true
   )
   &SolveForcesAndVelocity(Clamp Velocity=false, Limit Acceleration=false)
   &ScaleColor(
    Scale Alpha=&FloatFromCurve(FloatCurve=DataInterface() {propertyValues {Curve {keys [
         {tangentMode=RCTM_User, time=0, value=1}
         {arriveTangent=-4.3986, interpMode=RCIM_Cubic, leaveTangent=-4.3986, tangentMode=RCTM_User, time=0.45, value=1}
         {arriveTangent=0.1205, interpMode=RCIM_Cubic, leaveTangent=0.1205, tangentMode=RCTM_User, time=1, value=0}
        ]}}})
    Scale RGB=&Lerp_Vector(
     A=&VectorFromFloat(Value=&FloatFromCurve(FloatCurve=DataInterface() {propertyValues {Curve {keys [<at 0.2771 1>,<at 0.3786 1>]}}}, Scale Curve=40))
     B=&VectorFromFloat(Value=&RandomRangeFloat(Minimum=0, Maximum=2, Evaluation Type=::NewEnumerator1))
     Alpha=&FloatFromCurve(FloatCurve=DataInterface() {propertyValues {Curve {keys [
          {interpMode=RCIM_Cubic, leaveTangent=-0.3196, tangentMode=RCTM_Break, time=0, value=0}
          {arriveTangent=6.19, interpMode=RCIM_Cubic, tangentMode=RCTM_Break, time=0.4, value=1}
         ]}}})
    )
    ScaleA=true
    ScaleRGB=true
   )
   &ScaleColor(Scale RGB=&VectorFromFloat(Value=1), ScaleA=true, ScaleRGB=true) {module=ScaleColor002}
  ]
  rendererClasses [&NiagaraMeshRendererProperties]
  renderers [
   {rendererClass=&NiagaraMeshRendererProperties, rendererIndex=0}
   {rendererClass=&NiagaraMeshRendererProperties, rendererIndex=1}
  ]
 }
 Lighting_Arcs_Lingering {
  emitterUpdateScript [
   &EmitterState(Life Cycle Mode=::NewEnumerator1, Loop Behavior=::NewEnumerator1, Loop Delay=3.25, Loop Duration=0.15, UseLoopDelay=true)
   &SpawnRate(SpawnRate=200, Use Spawn Probability=false)
  ]
  particleSpawnScript [
   &InitializeParticle(
    AdjustAlpha=true
    AdjustHue=true
    AdjustSaturation=true
    AdjustValue=true
    Color=&MergeHelper_ColorSwitch(
     0=<rgba 1 0.2817 0 1>
     1=<rgba 1 0.7667 0 1>
     2=?
     3=?
     Index=&RandomRangeInt(Minimum=0, Maximum=1)
     Num=::NewEnumerator0
    )
    Color Mode=::NewEnumerator3
    Hue Shift Range:Vector2f=<-0.01 0.01>
    Lifetime Max=0.6
    Lifetime Min=0.3
    Lifetime Mode=::NewEnumerator1
    Mesh Scale Mode=::NewEnumerator3
    Mesh Uniform Scale=&RandomRangeFloat(Minimum=0.05, Maximum=0.07)
    UsePositionOffset=false
    Write Material Random=true
   )
   &ShapeLocation(MaskSpawnGroup=false, Override Seed=false, Sphere Radius=5, UseOwnerScale=true)
   &InitialMeshOrientation(
    Mesh Orientation Mode=::NewEnumerator7
    Rotation=&RandomRangeVector(Minimum=<-0.3 -0.3 -1>, Maximum=<0.3 0.3 1>)
    Use Rotation Vector=true
   )
   &InitialMeshOrientation(Mesh Orientation Mode=::NewEnumerator7, Rotation=<0 0.25 0>, Use Rotation Vector=true) {module=InitialMeshOrientation001}
   &CameraOffset(Camera Offset Amount=10)
  ]
  particleUpdateScript [
   &ParticleState()
   &SkeletalMeshLocation(
    Bone / Socket Index=&SelectIntFromArray(
     Array Sampling Mode=::NewEnumerator1
     Direct Array Index=?
     Int Selection Array=DataInterface() {propertyValues {IntData [10,11,12,13,14,15,25,26,27,28,29,38,39,40,41,42]}}
     Random Seed=0
    )
    Bone Sampling Mode=::NewEnumerator4
    MaskSpawnGroup=false
    Position Sampling=::NewEnumerator1
    Skeletal Mesh=$System.Skeletal Mesh Reader
   )
   &ScaleMeshSize(Scale Factor=&VectorFromFloat(Value=&FloatFromCurve(
      FloatCurve=DataInterface() {propertyValues {Curve {keys [
          {arriveTangent=1.9115, interpMode=RCIM_Cubic, leaveTangent=1.9115, tangentMode=RCTM_User, time=0, value=0}
          {arriveTangent=0.0224, interpMode=RCIM_Cubic, leaveTangent=0.0224, tangentMode=RCTM_User, time=1, value=1.5}
         ]}}}
      Scale Curve=1.25
     )))
   &ScaleMeshSize(Scale Factor=&RandomRangeVector(Minimum=<1 1 1.5>, Maximum=<1 1 2>)) {module=ScaleMeshSize001}
   &DynamicMaterialParameters(
    Index 0 Param 2=&RandomRangeFloat(Minimum=0, Maximum=0)
    Index 0 Param 4=13
    Index 1 Param 1=&FloatFromCurve(FloatCurve=DataInterface() {propertyValues {Curve {keys [<at 0 0.5>,<at 1 1>]}}})
    Index 1 Param 2=2
    Index 1 Param 3=&FloatFromCurve(FloatCurve=DataInterface() {propertyValues {Curve {keys [{interpMode=RCIM_Cubic, tangentMode=RCTM_Break, time=0.5, value=-0.5}, <at 1 0.1>]}}})
    Index 2 Param 1=&Multiply_Float(
     A=&FloatFromCurve(
      FloatCurve=DataInterface() {propertyValues {Curve {keys [{interpMode=RCIM_Cubic, leaveTangent=8.1085, tangentMode=RCTM_Break, time=0.8568, value=1}, {tangentMode=RCTM_User, time=1, value=0}]}}}
      Scale Curve=0.3
     )
     B=&Divide_Float(
      A=&MakeFloatFromVector(Vector=$Particles.Scale, Channel=::NewEnumerator0)
      B=&RandomRangeFloat(Minimum=10, Maximum=15)
     )
    )
    Index 2 Param 2=&FloatFromCurve(
     FloatCurve=DataInterface() {propertyValues {Curve {keys [
         {interpMode=RCIM_Cubic, tangentMode=RCTM_Break, time=0.7, value=0}
         {arriveTangent=-0.0859, interpMode=RCIM_Cubic, tangentMode=RCTM_Break, time=1, value=-1}
        ]}}}
     Scale Curve=50
    )
    Index 2 Param 3=&Waveform(
     Global Input Position=&RandomRangeFloat(Minimum=0, Maximum=15)
     Global Phase=$Particles.Age
     "Waveform [1]"=::NewEnumerator12
     "[1] Phase Scale"=2
     "[1] Frequency"=1
     "[1] Amplitude Min/Max"=<0 1>
    )
    Index 2 Param 4=25
    Index 3 Param 1=&FloatFromCurve(
     FloatCurve=DataInterface() {propertyValues {Curve {keys [
         {interpMode=RCIM_Cubic, leaveTangent=-0.0025, tangentMode=RCTM_Break, time=0, value=0}
         {arriveTangent=1.8528, interpMode=RCIM_Cubic, tangentMode=RCTM_Break, time=1, value=1}
        ]}}}
     Scale Curve=&RandomRangeFloat(Minimum=0, Maximum=0)
    )
    Index 3 Param 2=&FloatFromCurve(
     FloatCurve=DataInterface() {propertyValues {Curve {keys [
         {interpMode=RCIM_Cubic, leaveTangent=-0.0025, tangentMode=RCTM_Break, time=0, value=0.5}
         {arriveTangent=-0.0316, interpMode=RCIM_Cubic, tangentMode=RCTM_Break, time=1, value=-0.2}
        ]}}}
     Scale Curve=0
    )
    Index 3 Param 3=&FloatFromCurve(FloatCurve=DataInterface() {propertyValues {Curve {keys [
         {interpMode=RCIM_Cubic, leaveTangent=1.1034, tangentMode=RCTM_Break, time=0, value=-1}
         {arriveTangent=2.2399, tangentMode=RCTM_User, time=0.5, value=0.125}
        ]}}})
    Index 3 Param 4=&FloatFromCurve(FloatCurve=DataInterface() {propertyValues {Curve {keys [
         {interpMode=RCIM_Cubic, leaveTangent=1.8089, tangentMode=RCTM_Break, time=0.3, value=-0.5}
         {arriveTangent=2.0757, interpMode=RCIM_Cubic, tangentMode=RCTM_Break, time=0.9, value=1}
        ]}}})
    Module.Index0Param0Write=true
    Module.Index0Param1Write=true
    Module.Index0Param2Write=true
    Module.Index0Param3Write=true
    Module.Index1Param0Write=true
    Module.Index1Param1Write=true
    Module.Index1Param2Write=true
    Module.Index1Param3Write=true
    Module.Index2Param0Write=true
    Module.Index2Param1Write=true
    Module.Index2Param2Write=true
    Module.Index2Param3Write=true
    Module.Index3Param0Write=true
    Module.Index3Param1Write=true
    Module.Index3Param2Write=true
    Module.Index3Param3Write=true
    Write Parameter Index 1=true
    Write Parameter Index 2=true
    Write Parameter Index 3=true
   )
   &SolveForcesAndVelocity(Clamp Velocity=false, Limit Acceleration=false)
   &ScaleColor(
    Scale Alpha=&FloatFromCurve(FloatCurve=DataInterface() {propertyValues {Curve {keys [
         {tangentMode=RCTM_User, time=0, value=1}
         {arriveTangent=-4.3986, interpMode=RCIM_Cubic, leaveTangent=-4.3986, tangentMode=RCTM_User, time=0.45, value=1}
         {arriveTangent=0.1205, interpMode=RCIM_Cubic, leaveTangent=0.1205, tangentMode=RCTM_User, time=1, value=0}
        ]}}})
    Scale RGB=&Lerp_Vector(
     A=&VectorFromFloat(Value=&FloatFromCurve(FloatCurve=DataInterface() {propertyValues {Curve {keys [<at 0.2771 1>,<at 0.3786 1>]}}}, Scale Curve=40))
     B=&VectorFromFloat(Value=&RandomRangeFloat(Minimum=0, Maximum=2, Evaluation Type=::NewEnumerator1))
     Alpha=&FloatFromCurve(FloatCurve=DataInterface() {propertyValues {Curve {keys [
          {interpMode=RCIM_Cubic, leaveTangent=-0.3196, tangentMode=RCTM_Break, time=0, value=0}
          {arriveTangent=6.19, interpMode=RCIM_Cubic, tangentMode=RCTM_Break, time=0.4, value=1}
         ]}}})
    )
    ScaleA=true
    ScaleRGB=true
   )
  ]
  rendererClasses [&NiagaraMeshRendererProperties]
  renderers [
   {rendererClass=&NiagaraMeshRendererProperties, rendererIndex=0}
   {rendererClass=&NiagaraMeshRendererProperties, rendererIndex=1}
  ]
 }
 Spark_Pops {
  emitterUpdateScript [
   &EmitterState(
    Inactive Response=::NewEnumerator1
    Life Cycle Mode=::NewEnumerator1
    Loop Behavior=::NewEnumerator1
    Loop Delay=8.15
    Loop Duration=1.7
    UseLoopDelay=true
   )
   &SpawnRate(
    SpawnRate=&Multiply_Float(
     A=200
     B=&Multiply_Float(
      A=&FloatFromCurve(
       FloatCurve=DataInterface() {propertyValues {Curve {keys [{arriveTangent=-182.5864, interpMode=RCIM_Cubic, leaveTangent=-182.5864, tangentMode=RCTM_User, time=0, value=20}, <at 0.3 0>]}}}
       CurveIndex=&Frac_Float(Float=&Multiply_Float(A=&ClampFloat(Float=$Emitter.NormalizedLoopAge, Max=0.99), B=7))
      )
      B=1
     )
    )
    Use Spawn Probability=false
   )
  ]
  particleSpawnScript [
   &PartitionParticles()
   &InitializeParticle(
    Lifetime=&MergeHelper_FloatSwitch(
     0=&RandomRangeFloat(Minimum=0.05, Maximum=0.15)
     1=&RandomRangeFloat(Minimum=0.2, Maximum=0.25)
     2=?
     3=?
     Index=$Particles.PartitionParticles.Partition
     Num=::NewEnumerator0
    )
    Sprite Size=&Multiply_Vector2D_ByFloat(
     Float=&Multiply_Float(A=$System.FrontEndScale, B=&RandomRangeFloat(Minimum=4.5, Maximum=5))
     Vector2D=<1 1.5>
    )
    Sprite Size Mode=::NewEnumerator4
    UsePositionOffset=false
    Write Material Random=true
   )
   &ShapeLocation(MaskSpawnGroup=false, Override Seed=false, Rotation Mode=::NewEnumerator1, Sphere Radius=1, UseOwnerScale=true)
   &CameraOffset(Camera Offset Amount=5)
   &AddVelocityFromPoint(
    Module.Use Falloff Distance=false
    Velocity Strength=&Multiply_Float(
     A=&RandomRangeFloat(Minimum=125, Maximum=200)
     B=&FN_FrontEndFloatModifier(
      Gameplay=&MergeHelper_FloatSwitch(0=1.5, 1=0.75, 2=?, 3=?, Index=$Particles.PartitionParticles.Partition, Num=::NewEnumerator0)
      Front End=&Multiply_Float(
       A=&MergeHelper_FloatSwitch(0=1.5, 1=0.75, 2=?, 3=?, Index=$Particles.PartitionParticles.Partition, Num=::NewEnumerator0)
       B=$System.FrontEndScale
      )
     )
    )
   )
  ]
  particleUpdateScript [
   &ParticleState()
   &SkeletalMeshLocation(
    Bone / Socket Index=&MergeHelper_IntSwitch(
     0=0
     1=1
     2=2
     3=3
     4=4
     5=5
     6=6
     Index=&MakeIntFromFloat(
      Float=&Multiply_Float(A=&ClampFloat(Float=$Emitter.NormalizedLoopAge, Max=0.99), B=7)
      Conversion Mode=::NewEnumerator3
     )
     Num=::NewEnumerator5
    )
    MaskSpawnGroup=false
    Mesh Sampling Type=::NewEnumerator5
    Position Sampling=::NewEnumerator1
    Skeletal Mesh=$System.Skeletal Mesh Reader
    Socket Sampling Mode=::NewEnumerator5
   ) {module=SkeletalMeshLocation001}
   &AccelerationForce(Acceleration=<0 0 -500>)
   &Drag(Drag=&RandomRangeFloat(Minimum=4, Maximum=6), Use Linear Drag=true, Use Rotational Drag=true)
   &GravityForce(Gravity=<0 0 -200>)
   &SolveForcesAndVelocity(Clamp Velocity=false, Limit Acceleration=false)
   &ScaleSpriteSizeBySpeed(Max Scale Factor=<0.2 5>, Min Scale Factor=<0.2 0.2>)
   &ScaleSpriteSize(Uniform Curve Sprite Scale=DataInterface() {propertyValues {Curve {keys [{arriveTangent=-6.725, interpMode=RCIM_Cubic, leaveTangent=-6.725, tangentMode=RCTM_User, time=0.65, value=1}, <at 1 0 RCIM_Cubic>]}}})
   &ScaleColor(
    Scale RGB=&Multiply_VectorByFloat(
     Vector=&MakeVectorFromLinearColorRGB(Color=&MergeHelper_ColorSwitch(
       0=<rgba 0 0.1584 1 1>
       1=<rgba 1 0.7667 0 1>
       2=<rgba 1 0.7667 0 1>
       3=<rgba 1 0.0508 0 1>
       4=<rgba 1 0.0508 0 1>
       5=<rgba 1 0.7667 0 1>
       6=<rgba 0 0.1584 1 1>
       Index=&MakeIntFromFloat(Float=&Multiply_Float(A=$Emitter.NormalizedLoopAge, B=7), Conversion Mode=::NewEnumerator3)
       Num=::NewEnumerator5
      ))
     Float=&FloatFromCurve(FloatCurve=DataInterface() {propertyValues {Curve {keys [{arriveTangent=-23.4512, interpMode=RCIM_Cubic, leaveTangent=-23.4512, tangentMode=RCTM_User, time=0, value=10}, <at 1 1>]}}})
    )
    ScaleA=true
    ScaleRGB=true
   )
   /SetVariables(
    Particles.GlowColor=&MakeLinearColorFromVectorAndFloat(
     "Vector (RGB)"=&Multiply_VectorByFloat(
      Vector=&MakeVectorFromLinearColorRGB(Color=&MergeHelper_ColorSwitch(
        0=<rgba 0 0.1584 1 1>
        1=<rgba 1 0.7667 0 1>
        2=<rgba 1 0.7667 0 1>
        3=<rgba 1 0.0508 0 1>
        4=<rgba 1 0.0508 0 1>
        5=<rgba 1 0.7667 0 1>
        6=<rgba 0 0.1584 1 1>
        Index=&MakeIntFromFloat(
         Float=&Multiply_Float(A=&ClampFloat(Float=$Emitter.NormalizedLoopAge, Max=0.99), B=7)
         Conversion Mode=::NewEnumerator3
        )
        Num=::NewEnumerator5
       ))
      Float=&FloatFromCurve(FloatCurve=DataInterface() {propertyValues {Curve {keys [{arriveTangent=-23.4512, interpMode=RCIM_Cubic, leaveTangent=-23.4512, tangentMode=RCTM_User, time=0, value=10}, <at 1 1>]}}})
     )
     "Float (Alpha)"=&FloatFromCurve(FloatCurve=DataInterface() {propertyValues {Curve {keys [<at 0.5 1>,<at 1 0>]}}}, Scale Curve=0.03)
    )
    Particles.GlowSize=&Multiply_Vector2D(A=$Particles.SpriteSize, B=<8 2>)
   ) {isSetParameters=true}
  ]
  rendererClasses [&NiagaraSpriteRendererProperties]
  renderers [
   {rendererClass=&NiagaraSpriteRendererProperties}
   {rendererClass=&NiagaraSpriteRendererProperties, rendererIndex=1}
  ]
 }
 String_Ball_Decals {
  emitterUpdateScript [
   &EmitterState(Inactive Response=::NewEnumerator1, Life Cycle Mode=::NewEnumerator1, Loop Behavior=::NewEnumerator1, Loop Duration=2.8, UseLoopDelay=false)
   &SpawnRate(
    SpawnRate=&FloatFromCurve(
     FloatCurve=DataInterface() {propertyValues {Curve {
        defaultValue=1
        keys [<at 0 2 RCIM_Cubic>, {arriveTangent=24.8914, interpMode=RCIM_Cubic, leaveTangent=24.8914, tangentMode=RCTM_User, time=1, value=15}]
       }}}
     CurveIndex=$Emitter.NormalizedLoopAge
    )
    Use Spawn Probability=false
   )
   &SpawnBurst_Instantaneous(Spawn Group=1, Use Spawn Probability=false, UseLoopCountLimit=false)
  ]
  particleSpawnScript [
   /SetVariables(Particles.VisibilityTag=$Emitter.SpawnGroup) {isSetParameters=true}
   &InitializeParticle(
    Lifetime=&MergeHelper_FloatSwitch(
     0=&RandomRangeFloat(Minimum=0.3, Maximum=0.4)
     1=2.8
     2=?
     3=?
     Index=$Emitter.SpawnGroup
     Num=::NewEnumerator0
    )
    Sprite Size Mode=::NewEnumerator3
    Uniform Sprite Size=&MergeHelper_FloatSwitch(
     0=&RandomRangeFloat(Minimum=30, Maximum=40)
     1=120
     2=?
     3=?
     Index=$Emitter.SpawnGroup
     Num=::NewEnumerator0
    )
    UsePositionOffset=false
    Write Material Random=true
   )
   &DynamicMaterialParameters(
    Index 0 Param 1=&MergeHelper_FloatSwitch(0=1, 1=0.5, 2=?, 3=?, Index=$Emitter.SpawnGroup, Num=::NewEnumerator0)
    Index 0 Param 3=1
    Module.Index0Float0Write=true
    Module.Index0Float1Write=true
    Module.Index0Float2Write=true
    Module.Index0Float3Write=true
   )
   &CameraOffset(Camera Offset Amount=18)
   &SkeletalMeshLocation(
    Bone / Socket Index=7
    MaskSpawnGroup=false
    Mesh Sampling Type=::NewEnumerator5
    Position Sampling=::NewEnumerator1
    Skeletal Mesh=$System.Skeletal Mesh Reader
    Socket Sampling Mode=::NewEnumerator5
   )
   &ShapeLocation(MaskSpawnGroup=false, Override Seed=false, Sphere Radius=3, Sphere Surface Distribution=1, UseOwnerScale=true)
  ]
  particleUpdateScript [
   &ParticleState()
   &ScaleColor(
    Scale Alpha=&FloatFromCurve(
     FloatCurve=DataInterface() {propertyValues {Curve {
        defaultValue=1
        keys [
         {interpMode=RCIM_Cubic, tangentMode=RCTM_User, time=0, value=0}
         {interpMode=RCIM_Cubic, tangentMode=RCTM_User, time=0.5, value=1}
         {tangentMode=RCTM_User, time=1, value=0}
        ]
       }}}
     Scale Curve=&Multiply_Float(
      A=0.1
      B=&Waveform(
       "Waveform [1]"=::NewEnumerator0
       "[1] Phase Scale"=&FloatFromCurve(FloatCurve=DataInterface() {propertyValues {Curve {keys [<at 0 1 RCIM_Cubic>, {arriveTangent=-0.2757, interpMode=RCIM_Cubic, leaveTangent=-0.2757, tangentMode=RCTM_User, time=1.0012, value=2}]}}})
       "[1] Frequency"=1
       "[1] Amplitude Min/Max"=<0 1>
      )
     )
    )
    Scale RGB=&MakeVectorFromLinearColorRGB(Color=&MergeHelper_ColorSwitch(
      0=<rgba 1 0.05 0 1>
      1=<rgba 0 1 0.0668 1>
      2=<rgba 0 0.1001 1 1>
      3=?
      Index=&RandomRangeInt(Minimum=0, Maximum=2)
      Num=::NewEnumerator1
     ))
    ScaleA=true
    ScaleRGB=true
   ) {module=ScaleColor001}
   &FN_Light_Attributes(
    Color=$Particles.Color
    Create Color Attribute=true
    Create Exponent Attribute=true
    Create Position Attribute=true
    Create Radius Attribute=true
    Exponent=1
    Position=&AddVectorToPosition(Position=$Particles.Position, Vector=<0 0 15>)
    Radius=15
    Scale Color=3
   )
   /SetVariables(Particles.MainDecal_Pos:NiagaraPosition=$Particles.SkeletalMeshLocation.SampledPosition) {isSetParameters=true}
   /SetVariables(Particles.Main Decal Color:LinearColor=&MakeLinearColorFromVectorAndFloat(
     "Vector (RGB)"=&MakeVectorFromLinearColorRGB(Color=<rgba 1 0.7688 0 0.0001>)
     "Float (Alpha)"=&Multiply_Float(
      A=&FloatFromCurve(
       FloatCurve=DataInterface() {propertyValues {Curve {
          defaultValue=0.5
          keys [
           {interpMode=RCIM_Cubic, tangentMode=RCTM_User, time=0, value=1}
           {arriveTangent=4.6032, interpMode=RCIM_Cubic, leaveTangent=4.6032, tangentMode=RCTM_User, time=1, value=3}
          ]
         }}}
       Scale Curve=0.008
      )
      B=&Waveform(
       "Waveform [1]"=::NewEnumerator0
       "[1] Phase Scale"=&FloatFromCurve(FloatCurve=DataInterface() {propertyValues {Curve {keys [<at 0 1 RCIM_Cubic>, {arriveTangent=4.308, interpMode=RCIM_Cubic, leaveTangent=4.308, tangentMode=RCTM_User, time=1.0012, value=4}]}}})
       "[1] Frequency"=1
       "[1] Amplitude Min/Max"=<0 1>
      )
     )
    )) {isSetParameters=true}
  ]
  rendererClasses [&NiagaraSpriteRendererProperties]
  renderers [
   {rendererClass=&NiagaraSpriteRendererProperties}
   {rendererClass=&NiagaraSpriteRendererProperties, rendererIndex=1}
  ]
 }
}
summary {
 systemName=NS_Emote_DolhpinGill
 emitters [
  {emitterName=Appear_String_Lighting, rendererClasses [&NiagaraRibbonRendererProperties]}
  {emitterName=Disappear_String_Lighting, rendererClasses [&NiagaraRibbonRendererProperties]}
  {emitterName=LightingArcs_Appear, rendererClasses [&NiagaraMeshRendererProperties]}
  {emitterName=Lighting_Arcs_Lingering, rendererClasses [&NiagaraMeshRendererProperties]}
  {emitterName=Light_Pops, rendererClasses [&NiagaraSpriteRendererProperties, &NiagaraLightRendererProperties]}
  {emitterName=Spark_Pops, rendererClasses [&NiagaraSpriteRendererProperties]}
  {emitterName=Lighting_Arcs_Disappear001, rendererClasses [&NiagaraMeshRendererProperties]}
  {emitterName=String_Ball_Decals, rendererClasses [&NiagaraSpriteRendererProperties]}
  {emitterName=Lighting_Arcs_Buildup, rendererClasses [&NiagaraMeshRendererProperties]}
  {emitterName=EndLighting, rendererClasses [&NiagaraMeshRendererProperties]}
 ]
 userVariables [
  User.bIsFrontEnd {type=NiagaraBool}
  User.bIsFrontEndPreview {type=NiagaraBool}
 ]
}
