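# One-shot Niagara burst effect for the DolphinGill emote, layering a brief orange glow with light, orange-yellow mesh arcs, faint smoke and multicolored sparks.
# Short and lingering spark emitters share drag, gravity and speed-stretched sprites, writing separate glow color and size attributes.
# The arcs animate through four dynamic material parameter sets.
# Front-end flags and scale adjust smoke and sparks, while the glow varies with time of day and front-end context.
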NiagaraSystem /BRCosmetics/Animation/Game/MainPlayer/Emotes/DolphinGill/FX/NS_Emote_DolhpinGill__Burst
dependencies {
 usedDataInterfaces [&NiagaraDataInterfaceCurve, &NiagaraDataInterfaceVectorCurve]
 usedDynamicInputs [&Vector2DFromFloat, &Multiply_VectorByFloat]
 usedModules [
  &self:SystemSpawnScript.SystemScriptSource.SystemScriptGraph.NiagaraNodeAssignment_0.SetVariables_#1
  &SystemState
 ]
}
emitters {
 Glow {
  emitterUpdateScript [
   &EmitterState()
   &SpawnBurst_Instantaneous(Use Spawn Probability=false, UseLoopCountLimit=false)
  ]
  particleSpawnScript [
   &InitializeParticle(
    Color=<rgba 1 0.5845 0 0.7>
    Lifetime=0.15
    Position Offset=<0 0 -5>
    Sprite Size Mode=::NewEnumerator3
    Uniform Sprite Size=200
    UsePositionOffset=true
    Write Material Random=true
   )
   &DynamicMaterialParameters(
    Index 0 Param 1=0.2
    Index 0 Param 2=1
    Module.Param0WriteEnabled=true
    Module.Param1WriteEnabled=true
    Module.Param2WriteEnabled=true
    Module.Param3WriteEnabled=true
   )
   &CameraOffset(Camera Offset Amount=18.5431)
  ]
  particleUpdateScript [
   &ParticleState()
   &ScaleColor(
    Scale Alpha=&Multiply_Float(
     A=&FNTimeOfDayScalar(Night Scale=0.5, Morning Scale=0.7, Evening Scale=0.7)
     B=&FloatFromCurve(FloatCurve=DataInterface() {propertyValues {Curve {
         defaultValue=1
         keys [{interpMode=RCIM_Cubic, tangentMode=RCTM_User, time=0, value=1}, {interpMode=RCIM_Cubic, tangentMode=RCTM_User, time=1, value=0}]
        }}})
    )
    Scale RGB=&VectorFromFloat(Value=&FN_FrontEndFloatModifier(Gameplay=1.5, Front End=1))
    ScaleA=true
    ScaleRGB=true
   )
   &FN_Light_Attributes(
    Color=$Particles.Color
    Create Color Attribute=true
    Create Exponent Attribute=true
    Create Position Attribute=true
    Create Radius Attribute=true
    Exponent=5
    Position=<15 0 0>
    Radius=90
    Scale Color=&FloatFromCurve(
     FloatCurve=DataInterface() {propertyValues {Curve {keys [
         {interpMode=RCIM_Cubic, tangentMode=RCTM_User, time=0, value=1}
         {arriveTangent=-0.5426, interpMode=RCIM_Cubic, leaveTangent=-0.5426, tangentMode=RCTM_User, time=1, value=0}
        ]}}}
     Scale Curve=5
    )
   )
  ]
  rendererClasses [&NiagaraSpriteRendererProperties, &NiagaraLightRendererProperties]
  renderers [
   {rendererClass=&NiagaraSpriteRendererProperties}
   {rendererClass=&NiagaraLightRendererProperties, rendererIndex=1}
  ]
 }
 Lighting_Arcs {
  emitterUpdateScript [
   &EmitterState()
   &SpawnBurst_Instantaneous(
    Spawn Count=&RandomRangeInt(Minimum=10, Maximum=13)
    Use Spawn Probability=false
    UseLoopCountLimit=false
   )
  ]
  particleSpawnScript [
   &InitializeParticle(
    AdjustAlpha=true
    AdjustHue=true
    AdjustSaturation=true
    AdjustValue=true
    Color=&MergeHelper_ColorSwitch(
     0=<rgba 1 0.2817 0 1>
     1=<rgba 1 0.7667 0 1>
     2=?
     3=?
     Index=&RandomRangeInt(Minimum=0, Maximum=1)
     Num=::NewEnumerator0
    )
    Color Mode=::NewEnumerator3
    Hue Shift Range:Vector2f=<-0.01 0.01>
    Lifetime Max=0.6
    Lifetime Min=0.3
    Lifetime Mode=::NewEnumerator1
    Mesh Scale Mode=::NewEnumerator3
    Mesh Uniform Scale=&RandomRangeFloat(Minimum=0.09, Maximum=0.11)
    UsePositionOffset=false
    Write Material Random=true
   )
   &ShapeLocation(MaskSpawnGroup=false, Offset=<0 0 -2>, Offset Mode=::NewEnumerator0, Override Seed=false, Sphere Radius=7, UseOwnerScale=true)
   &InitialMeshOrientation(
    Mesh Orientation Mode=::NewEnumerator7
    Rotation=&RandomRangeVector(Minimum=<-0.3 -0.3 -1>, Maximum=<0.3 0.3 1>)
    Use Rotation Vector=true
   )
   &InitialMeshOrientation(Mesh Orientation Mode=::NewEnumerator7, Rotation=<0 0.25 0>, Use Rotation Vector=true) {module=InitialMeshOrientation001}
   &CameraOffset(Camera Offset Amount=5)
   &AddVelocity(
    Module.Use Falloff Distance=false
    Override Seed=false
    Velocity Mode=::NewEnumerator1
    Velocity Speed=&Multiply_Float(
     A=&RandomRangeFloat(Minimum=30, Maximum=50)
     B=&Waveform(
      "Waveform [1]"=::NewEnumerator5
      "[1] Phase Scale"=&FloatFromCurve(
       FloatCurve=DataInterface() {propertyValues {Curve {keys [<at 0 1 RCIM_Cubic>, {arriveTangent=4.308, interpMode=RCIM_Cubic, leaveTangent=4.308, tangentMode=RCTM_User, time=1.0012, value=4}]}}}
       CurveIndex=$Emitter.NormalizedLoopAge
      )
      "[1] Frequency"=1
      "[1] Amplitude Min/Max"=<0 1>
     )
    )
   )
  ]
  particleUpdateScript [
   &ParticleState()
   &ScaleMeshSize(Scale Factor=&VectorFromFloat(Value=&FloatFromCurve(
      FloatCurve=DataInterface() {propertyValues {Curve {keys [
          {arriveTangent=1.9115, interpMode=RCIM_Cubic, leaveTangent=1.9115, tangentMode=RCTM_User, time=0, value=0}
          {arriveTangent=0.0224, interpMode=RCIM_Cubic, leaveTangent=0.0224, tangentMode=RCTM_User, time=1, value=1.5}
         ]}}}
      Scale Curve=1.25
     )))
   &DynamicMaterialParameters(
    Index 0 Param 2=&RandomRangeFloat(Minimum=0, Maximum=0)
    Index 0 Param 4=13
    Index 1 Param 1=&FloatFromCurve(FloatCurve=DataInterface() {propertyValues {Curve {keys [<at 0 0.5>,<at 1 1>]}}})
    Index 1 Param 2=2
    Index 1 Param 3=&FloatFromCurve(FloatCurve=DataInterface() {propertyValues {Curve {keys [{interpMode=RCIM_Cubic, tangentMode=RCTM_Break, time=0.5, value=-0.5}, <at 1 0.1>]}}})
    Index 2 Param 1=&Multiply_Float(
     A=&FloatFromCurve(
      FloatCurve=DataInterface() {propertyValues {Curve {keys [{interpMode=RCIM_Cubic, leaveTangent=8.1085, tangentMode=RCTM_Break, time=0.8568, value=1}, {tangentMode=RCTM_User, time=1, value=0}]}}}
      Scale Curve=0.3
     )
     B=&Divide_Float(
      A=&MakeFloatFromVector(Vector=$Particles.Scale, Channel=::NewEnumerator0)
      B=&RandomRangeFloat(Minimum=10, Maximum=15)
     )
    )
    Index 2 Param 2=&FloatFromCurve(
     FloatCurve=DataInterface() {propertyValues {Curve {keys [
         {interpMode=RCIM_Cubic, tangentMode=RCTM_Break, time=0.7, value=0}
         {arriveTangent=-0.0859, interpMode=RCIM_Cubic, tangentMode=RCTM_Break, time=1, value=-1}
        ]}}}
     Scale Curve=50
    )
    Index 2 Param 3=&Waveform(
     Global Input Position=&RandomRangeFloat(Minimum=0, Maximum=15)
     Global Phase=$Particles.Age
     "Waveform [1]"=::NewEnumerator12
     "[1] Phase Scale"=2
     "[1] Frequency"=1
     "[1] Amplitude Min/Max"=<0 1>
    )
    Index 2 Param 4=25
    Index 3 Param 1=&FloatFromCurve(
     FloatCurve=DataInterface() {propertyValues {Curve {keys [
         {interpMode=RCIM_Cubic, leaveTangent=-0.0025, tangentMode=RCTM_Break, time=0, value=0}
         {arriveTangent=1.8528, interpMode=RCIM_Cubic, tangentMode=RCTM_Break, time=1, value=1}
        ]}}}
     Scale Curve=&RandomRangeFloat(Minimum=0, Maximum=0)
    )
    Index 3 Param 2=&FloatFromCurve(
     FloatCurve=DataInterface() {propertyValues {Curve {keys [
         {interpMode=RCIM_Cubic, leaveTangent=-0.0025, tangentMode=RCTM_Break, time=0, value=0.5}
         {arriveTangent=-0.0316, interpMode=RCIM_Cubic, tangentMode=RCTM_Break, time=1, value=-0.2}
        ]}}}
     Scale Curve=0
    )
    Index 3 Param 3=&FloatFromCurve(FloatCurve=DataInterface() {propertyValues {Curve {keys [
         {interpMode=RCIM_Cubic, leaveTangent=1.1034, tangentMode=RCTM_Break, time=0, value=-1}
         {arriveTangent=2.2399, tangentMode=RCTM_User, time=0.5, value=0.125}
        ]}}})
    Index 3 Param 4=&FloatFromCurve(FloatCurve=DataInterface() {propertyValues {Curve {keys [
         {interpMode=RCIM_Cubic, leaveTangent=1.8089, tangentMode=RCTM_Break, time=0.3, value=-0.5}
         {arriveTangent=2.0757, interpMode=RCIM_Cubic, tangentMode=RCTM_Break, time=0.9, value=1}
        ]}}})
    Module.Index0Param0Write=true
    Module.Index0Param1Write=true
    Module.Index0Param2Write=true
    Module.Index0Param3Write=true
    Module.Index1Param0Write=true
    Module.Index1Param1Write=true
    Module.Index1Param2Write=true
    Module.Index1Param3Write=true
    Module.Index2Param0Write=true
    Module.Index2Param1Write=true
    Module.Index2Param2Write=true
    Module.Index2Param3Write=true
    Module.Index3Param0Write=true
    Module.Index3Param1Write=true
    Module.Index3Param2Write=true
    Module.Index3Param3Write=true
    Write Parameter Index 1=true
    Write Parameter Index 2=true
    Write Parameter Index 3=true
   )
   &Drag(Drag=3, Rotational Drag=0, Use Linear Drag=true, Use Rotational Drag=true)
   &InheritSourceMovement(ApplyForce=false, ApplyPosition=true, ApplyPositionFalloff=false)
   &SolveForcesAndVelocity(Clamp Velocity=false, Limit Acceleration=false)
   &ScaleColor(
    Scale Alpha=&FloatFromCurve(FloatCurve=DataInterface() {propertyValues {Curve {keys [
         {tangentMode=RCTM_User, time=0, value=1}
         {arriveTangent=-4.3986, interpMode=RCIM_Cubic, leaveTangent=-4.3986, tangentMode=RCTM_User, time=0.45, value=1}
         {arriveTangent=0.1205, interpMode=RCIM_Cubic, leaveTangent=0.1205, tangentMode=RCTM_User, time=1, value=0}
        ]}}})
    Scale RGB=&Lerp_Vector(
     A=&VectorFromFloat(Value=&FloatFromCurve(FloatCurve=DataInterface() {propertyValues {Curve {keys [<at 0.2771 1>,<at 0.3786 1>]}}}, Scale Curve=40))
     B=&VectorFromFloat(Value=&RandomRangeFloat(Minimum=0, Maximum=2, Evaluation Type=::NewEnumerator1))
     Alpha=&FloatFromCurve(FloatCurve=DataInterface() {propertyValues {Curve {keys [
          {interpMode=RCIM_Cubic, leaveTangent=-0.3196, tangentMode=RCTM_Break, time=0, value=0}
          {arriveTangent=6.19, interpMode=RCIM_Cubic, tangentMode=RCTM_Break, time=0.4, value=1}
         ]}}})
    )
    ScaleA=true
    ScaleRGB=true
   )
  ]
  rendererClasses [&NiagaraMeshRendererProperties]
  renderers [
   {rendererClass=&NiagaraMeshRendererProperties, rendererIndex=0}
   {rendererClass=&NiagaraMeshRendererProperties, rendererIndex=1}
  ]
 }
 Smoke {
  emitterUpdateScript [
   &EmitterState()
   &SpawnBurst_Instantaneous(
    Spawn Count=&RandomRangeInt(Minimum=2, Maximum=3)
    Use Spawn Probability=false
    UseLoopCountLimit=false
   )
  ]
  particleSpawnScript [
   &InitializeParticle(
    Color Maximum=<rgba 0.76 1 1 0.2>
    Color Minimum=<rgba 0.8 0.917 1 0.15>
    Color Mode=::NewEnumerator2
    Lifetime Max=0.7
    Lifetime Min=0.4
    Lifetime Mode=::NewEnumerator1
    Sprite Size Mode=::NewEnumerator3
    Uniform Sprite Size=&Multiply_Float(
     A=&MakeCustomFloatFromBool(Bool=$User.bIsFrontEnd, True Float=0.45, False Float=0.75)
     B=&RandomRangeFloat(Minimum=100, Maximum=130)
    )
    UsePositionOffset=false
    Write Material Random=true
   )
   &CameraOffset(Camera Offset Amount=&RandomRangeFloat(Minimum=20, Maximum=30, Evaluation Type=::NewEnumerator1))
   &ShapeLocation(
    Hemisphere Distribution=<1 0.5>
    MaskSpawnGroup=false
    Offset=<0 0 -10>
    Offset Coordinate Space=::World
    Offset Mode=::NewEnumerator0
    Override Seed=false
    Rotation Angle=90
    Rotation Mode=::NewEnumerator1
    Sphere Radius=&MakeCustomFloatFromBool(Bool=$User.bIsFrontEnd, True Float=5, False Float=1)
    UseOwnerScale=true
   )
   &AddVelocity(
    Module.Use Falloff Distance=false
    Origin Offset Coordinate Space=::World
    Override Seed=false
    Velocity Mode=::NewEnumerator1
    Velocity Speed=20
   )
  ]
  particleUpdateScript [
   &ParticleState()
   &ScaleSpriteSize(Uniform Curve Sprite Scale=DataInterface() {propertyValues {Curve {
       defaultValue=0.25
       keys [
        {arriveTangent=4.1785, interpMode=RCIM_Cubic, leaveTangent=4.1785, tangentMode=RCTM_User, time=0, value=0}
        {arriveTangent=1.1179, interpMode=RCIM_Cubic, leaveTangent=1.1179, tangentMode=RCTM_User, time=0.2042, value=0.6228}
        {arriveTangent=0.0224, interpMode=RCIM_Cubic, leaveTangent=0.0224, tangentMode=RCTM_User, time=1, value=1}
       ]
      }}})
   &ScaleColor(
    Scale Alpha=&FloatFromCurve(
     FloatCurve=DataInterface() {propertyValues {Curve {keys [
         {arriveTangent=2.4358, interpMode=RCIM_Cubic, leaveTangent=2.4358, tangentMode=RCTM_User, time=0, value=0}
         <at 0.25 0.5 RCIM_Cubic>
         <at 1 0 RCIM_Cubic>
        ]}}}
     Scale Curve=0.6
    )
    Scale RGB=&VectorFromCurve(VectorCurve=DataInterface() {propertyValues {XCurve {keys [<at 0 1>,<at 1 1>]}, YCurve {keys [<at 0 1>,<at 1 1>]}, ZCurve {keys [<at 0 1>,<at 1 1>]}}})
    ScaleA=true
    ScaleRGB=true
   )
   &DynamicMaterialParameters(
    Index 0 Param 1=&Add_Float(
     A=&RandomRangeFloat(Minimum=0, Maximum=1)
     B=&FloatFromCurve(
      FloatCurve=DataInterface() {propertyValues {Curve {keys [
          <at 0 0 RCIM_Cubic>
          {arriveTangent=1.5947, interpMode=RCIM_Cubic, leaveTangent=1.5947, tangentMode=RCTM_User, time=0.2557, value=0.5257}
          <at 1 1 RCIM_Cubic>
         ]}}}
      CurveIndex=$Particles.Age
      Scale Curve=0.25
     )
    )
    Module.Param0WriteEnabled=true
    Module.Param1WriteEnabled=true
    Module.Param2WriteEnabled=true
    Module.Param3WriteEnabled=true
   ) {module=DynamicMaterialParameters001}
   &Drag(Drag=3, Use Linear Drag=true, Use Rotational Drag=true)
   &SolveForcesAndVelocity(Clamp Velocity=false, Limit Acceleration=false)
  ]
  rendererClasses [&NiagaraSpriteRendererProperties]
  renderers [
   {rendererClass=&NiagaraSpriteRendererProperties}
   {rendererClass=&NiagaraSpriteRendererProperties, rendererIndex=1}
  ]
 }
 Spark_Burst {
  emitterUpdateScript [
   &EmitterState()
   &SpawnBurst_Instantaneous(
    Spawn Count=50
    Spawn Probability=&MakeCustomFloatFromBool(Bool=$User.bIsFrontEndPreview, True Float=1, False Float=1)
    Use Spawn Probability=true
    UseLoopCountLimit=false
   )
  ]
  particleSpawnScript [
   &PartitionParticles()
   &InitializeParticle(
    Color=&MergeHelper_ColorSwitch(
     0=<rgba 1 0.0508 0 1>
     1=<rgba 0 1 0.0764 1>
     2=<rgba 1 0.7667 0 1>
     3=<rgba 0 0.1584 1 1>
     Index=&RandomRangeInt(Minimum=0, Maximum=3)
     Num=::NewEnumerator2
    )
    Lifetime=&MergeHelper_FloatSwitch(
     0=&RandomRangeFloat(Minimum=0.05, Maximum=0.4)
     1=&RandomRangeFloat(Minimum=0.3, Maximum=0.6)
     2=?
     3=?
     Index=$Particles.PartitionParticles.Partition
     Num=::NewEnumerator0
    )
    Sprite Size=&Multiply_Vector2D_ByFloat(
     Float=&Multiply_Float(A=$System.FrontEndScale, B=&RandomRangeFloat(Minimum=4.5, Maximum=5))
     Vector2D=<1 1.5>
    )
    Sprite Size Mode=::NewEnumerator4
    UsePositionOffset=false
    Write Material Random=true
   )
   &ShapeLocation(MaskSpawnGroup=false, Override Seed=false, Rotation Mode=::NewEnumerator1, Sphere Radius=4.6628, UseOwnerScale=true)
   &CameraOffset(Camera Offset Amount=5)
   &AddVelocityFromPoint(
    Module.Use Falloff Distance=false
    Origin Offset=<0 0 -4>
    Origin Offset Coordinate Space=::World
    Velocity Strength=&Multiply_Float(
     A=&RandomRangeFloat(Minimum=150, Maximum=250)
     B=&Multiply_Float(
      A=&MergeHelper_FloatSwitch(0=1.5, 1=0.75, 2=?, 3=?, Index=$Particles.PartitionParticles.Partition, Num=::NewEnumerator0)
      B=$System.FrontEndScale
     )
    )
   )
  ]
  particleUpdateScript [
   &ParticleState()
   &Drag(Drag=&RandomRangeFloat(Minimum=4, Maximum=6), Use Linear Drag=true, Use Rotational Drag=true)
   &GravityForce(Gravity=<0 0 -300>)
   &SolveForcesAndVelocity(Clamp Velocity=false, Limit Acceleration=false)
   &ScaleSpriteSizeBySpeed(Max Scale Factor=<0.2 8>, Min Scale Factor=<0.2 0.2>)
   &ScaleSpriteSize(Uniform Curve Sprite Scale=DataInterface() {propertyValues {Curve {keys [{arriveTangent=-6.725, interpMode=RCIM_Cubic, leaveTangent=-6.725, tangentMode=RCTM_User, time=0.65, value=1}, <at 1 0 RCIM_Cubic>]}}})
   &ScaleColor(
    Scale RGB=&VectorFromFloat(Value=&FloatFromCurve(
      FloatCurve=DataInterface() {propertyValues {Curve {keys [{interpMode=RCIM_Cubic, tangentMode=RCTM_User, time=0, value=1}, {interpMode=RCIM_Cubic, tangentMode=RCTM_User, time=1, value=0}]}}}
      Scale Curve=10
     ))
    ScaleA=true
    ScaleRGB=true
   )
   /SetVariables(
    Particles.GlowColor=&MakeLinearColorFromVectorAndFloat(
     "Vector (RGB)"=&MakeVectorFromLinearColorRGB(Color=$Particles.Initial.Color)
     "Float (Alpha)"=&FloatFromCurve(FloatCurve=DataInterface() {propertyValues {Curve {keys [<at 0.5 1>,<at 1 0>]}}}, Scale Curve=0.15)
    )
    Particles.GlowSize=&Multiply_Vector2D(A=$Particles.SpriteSize, B=<8 2>)
   ) {isSetParameters=true, module=SetVariables_#2}
  ]
  rendererClasses [&NiagaraSpriteRendererProperties]
  renderers [
   {rendererClass=&NiagaraSpriteRendererProperties}
   {rendererClass=&NiagaraSpriteRendererProperties, rendererIndex=1}
  ]
 }
 Spark_Burst_Lingering {
  emitterUpdateScript [
   &EmitterState()
   &SpawnBurst_Instantaneous(
    Spawn Count=50
    Spawn Probability=&MakeCustomFloatFromBool(Bool=$User.bIsFrontEndPreview, True Float=1, False Float=1)
    Use Spawn Probability=true
    UseLoopCountLimit=false
   )
  ]
  particleSpawnScript [
   &PartitionParticles()
   &InitializeParticle(
    Color=&MergeHelper_ColorSwitch(
     0=<rgba 1 0.0508 0 1>
     1=<rgba 0 1 0.0764 1>
     2=<rgba 1 0.7667 0 1>
     3=<rgba 0 0.1584 1 1>
     Index=&RandomRangeInt(Minimum=0, Maximum=3)
     Num=::NewEnumerator2
    )
    Lifetime Max=1.7
    Lifetime Min=0.7
    Lifetime Mode=::NewEnumerator1
    Sprite Size=&Multiply_Vector2D_ByFloat(
     Float=&Multiply_Float(A=$System.FrontEndScale, B=&RandomRangeFloat(Minimum=4.5, Maximum=5))
     Vector2D=<1 1.5>
    )
    Sprite Size Mode=::NewEnumerator4
    UsePositionOffset=false
    Write Material Random=true
   )
   &ShapeLocation(MaskSpawnGroup=false, Override Seed=false, Rotation Mode=::NewEnumerator1, Sphere Radius=8, UseOwnerScale=true)
   &CameraOffset(Camera Offset Amount=5)
   &AddVelocityFromPoint(
    Module.Use Falloff Distance=false
    Origin Offset=<0 0 -7>
    Origin Offset Coordinate Space=::World
    Velocity Strength=&Multiply_Float(
     A=&RandomRangeFloat(Minimum=125, Maximum=300)
     B=&Multiply_Float(
      A=&MergeHelper_FloatSwitch(0=1.25, 1=0.75, 2=?, 3=?, Index=$Particles.PartitionParticles.Partition, Num=::NewEnumerator0)
      B=$System.FrontEndScale
     )
    )
   )
  ]
  particleUpdateScript [
   &ParticleState()
   &Drag(
    Drag=&FloatFromCurve(FloatCurve=DataInterface() {propertyValues {Curve {keys [<at 0 1>,<at 1 0>]}}}, Scale Curve=3)
    Use Linear Drag=true
    Use Rotational Drag=true
   )
   &GravityForce(Gravity=<0 0 -350>)
   &SolveForcesAndVelocity(Clamp Velocity=false, Limit Acceleration=false)
   &ScaleSpriteSizeBySpeed(Max Scale Factor=<0.2 8>, Min Scale Factor=<0.2 0.2>)
   &ScaleSpriteSize(Uniform Curve Sprite Scale=DataInterface() {propertyValues {Curve {keys [{arriveTangent=0.7853, interpMode=RCIM_Cubic, leaveTangent=0.7853, tangentMode=RCTM_User, time=0, value=1}, <at 0.8 0 RCIM_Cubic>]}}})
   &ScaleColor(
    Scale RGB=&VectorFromFloat(Value=&FloatFromCurve(
      FloatCurve=DataInterface() {propertyValues {Curve {keys [{interpMode=RCIM_Cubic, tangentMode=RCTM_User, time=0, value=1}, {interpMode=RCIM_Cubic, tangentMode=RCTM_User, time=1, value=0}]}}}
      Scale Curve=10
     ))
    ScaleA=true
    ScaleRGB=true
   )
   /SetVariables(
    Particles.GlowColor=&MakeLinearColorFromVectorAndFloat(
     "Vector (RGB)"=&MakeVectorFromLinearColorRGB(Color=$Particles.Initial.Color)
     "Float (Alpha)"=&FloatFromCurve(FloatCurve=DataInterface() {propertyValues {Curve {keys [<at 0.5 1>,<at 1 0>]}}}, Scale Curve=0.15)
    )
    Particles.GlowSize=&Multiply_Vector2D(A=$Particles.SpriteSize, B=<8 2>)
   ) {isSetParameters=true, module=SetVariables_#2}
  ]
  rendererClasses [&NiagaraSpriteRendererProperties]
  renderers [
   {rendererClass=&NiagaraSpriteRendererProperties}
   {rendererClass=&NiagaraSpriteRendererProperties, rendererIndex=1}
  ]
 }
}
summary {
 systemName=NS_Emote_DolhpinGill__Burst
 emitters [
  {emitterName=Spark_Burst, rendererClasses [&NiagaraSpriteRendererProperties]}
  {emitterName=Glow, rendererClasses [&NiagaraSpriteRendererProperties, &NiagaraLightRendererProperties]}
  {emitterName=Smoke, rendererClasses [&NiagaraSpriteRendererProperties]}
  {emitterName=Spark_Burst_Lingering, rendererClasses [&NiagaraSpriteRendererProperties]}
  {emitterName=Lighting_Arcs, rendererClasses [&NiagaraMeshRendererProperties]}
 ]
 userVariables [
  User.bIsFrontEnd {type=NiagaraBool}
  User.bIsFrontEndPreview {type=NiagaraBool}
 ]
}
